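# Control Rig for the F_MED_VowelSleep_Expand player body, correcting deformation of armpits, lapels, headphone wire, skirt and jacket.
# FortDeformRig runs first, then spherical pose readers cache thigh poses into variables that weight mirrored jacket and skirt offsets.
# Skirt bases follow twist-free thigh helper bones, with centre bases blended and aimed between sides.
# The wire uses weighted parent constraints and a stretched spline; def_simspace partially follows the pelvis.
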ControlRigBlueprint /BRCosmetics/Characters/Player/Female/Medium/Bodies/F_MED_VowelSleep_Expand/Meshes/F_MED_VowelSleep_Expand_CtrlRig
graphs [
 CollapseNode_ContainedGraph {nodes [
   n1 = RigVMFunctionEntryNode()
   /Script/ControlRig.RigUnit_PositionConstraintLocalSpaceOffset(
    Child=(Type=Bone,Name="deform_armpit_l")
    Filter=(bX=false,bY=True,bZ=True)
    Parents=((Item=(Type=Bone,Name="clavicle_l"),Weight=1),(Item=(Type=Bone,Name="upperarm_twist_01_l"),Weight=1))
    Weight=1
    bMaintainOffset=True
    ExecutePin=@n1.ExecutePin
   ) {is=struct RigVMUnitNode}
   > &RigUnit_PositionConstraintLocalSpaceOffset(
    Child=(Type=Bone,Name="deform_armpit_r")
    Filter=(bX=false,bY=True,bZ=True)
    Parents=((Item=(Type=Bone,Name="clavicle_r"),Weight=1),(Item=(Type=Bone,Name="upperarm_twist_01_r"),Weight=1))
    Weight=1
    bMaintainOffset=True
   )
   > n2 = &RigUnit_SphericalPoseReader(
    ActiveRegionScaleFactors=(PositiveWidth=2.4,NegativeWidth=2.4,PositiveHeight=2.8,NegativeHeight=1.5)
    ActiveRegionSize=0.1
    Debug=(bDrawDebug=true,bDraw2D=False,bDrawLocalAxes=False,DebugScale=25,DebugSegments=20,DebugThickness=0.25)
    DriverAxis=(X=1,Y=0,Z=0)
    DriverItem=(Type=Bone,Name="upperarm_l")
    FalloffRegionScaleFactors=(PositiveWidth=3,NegativeWidth=3,PositiveHeight=1,NegativeHeight=1)
    FalloffSize=0.28
    FlipHeightScaling=False
    FlipWidthScaling=False
    OptionalParentItem=(Type=Bone,Name="None")
    RotationOffset=(X=137.4003,Y=1.4417,Z=-67.9893)
   )
   > &RigUnit_ModifyTransforms(
    ItemToModify="((Item=(Type=Bone,Name=\"deform_armpit_l\"),Transform=(Rotation=(X=0,Y=0,Z=0,W=1),Translation=(X=5,Y=0,Z=0),Scale3D=(X=1,Y=1,Z=1))),(Item=(Type=Bone,Name=\"upperarm_twist_01_l\"),Transform=(Rotation=(X=0,Y=0,Z=0,W=1),Translation=(X=-1,Y=2,Z=0),Scale3D=(X=1,Y=1 ...(truncated)"
    WeightMinimum=0
    Weight=@n2.OutputParam
   )
   > n3 = &RigUnit_SphericalPoseReader(
    ActiveRegionScaleFactors=(PositiveWidth=2.4,NegativeWidth=2.4,PositiveHeight=2.8,NegativeHeight=1.5)
    ActiveRegionSize=0.1
    Debug=(bDrawDebug=true,bDraw2D=False,bDrawLocalAxes=False,DebugScale=25,DebugSegments=20,DebugThickness=0.25)
    DriverAxis=(X=-1,Y=0,Z=0)
    DriverItem=(Type=Bone,Name="upperarm_r")
    FalloffRegionScaleFactors=(PositiveWidth=3,NegativeWidth=3,PositiveHeight=1,NegativeHeight=1)
    FalloffSize=0.28
    FlipHeightScaling=False
    FlipWidthScaling=False
    OptionalParentItem=(Type=Bone,Name="None")
    RotationOffset=(X=-50,Y=1.4417,Z=-67.9893)
   )
   > n4 = &RigUnit_ModifyTransforms(
    ItemToModify="((Item=(Type=Bone,Name=\"deform_armpit_r\"),Transform=(Rotation=(X=0,Y=0,Z=0,W=1),Translation=(X=-5,Y=0,Z=0),Scale3D=(X=1,Y=1,Z=1))),(Item=(Type=Bone,Name=\"upperarm_twist_01_r\"),Transform=(Rotation=(X=0,Y=0,Z=0,W=1),Translation=(X=1,Y=2,Z=0),Scale3D=(X=1,Y=1 ...(truncated)"
    WeightMinimum=0
    Weight=@n3.OutputParam
   )
   RigVMFunctionReturnNode(ExecutePin=@n4.ExecutePin)
  ]}
 CollapseNode_ContainedGraph {nodes [
   n1 = RigVMFunctionEntryNode()
   n2 = RigVMDispatchNode(ExecuteContext=@n1.ExecutePin, Array=RigVMVariableNode() {variable=InSplineTransforms}.Value) {template="DISPATCH_RigVMDispatch_ArrayReset(io Array)"}
   > Set_Num = RigVMDispatchNode(Num=5, Array=@n2.Array) {template="DISPATCH_RigVMDispatch_ArraySetNum(io Array,in Num)"}
   > &RigUnit_ParentConstraint(
    AdvancedSettings=(RotationOrderForFilter=XZY)
    Child=(Type=Bone,Name="CR_Spline_Headphone_Wire_02")
    Parents=((Item=(Type=Bone,Name="dyn_headphones"),Weight=0.3),(Item=(Type=Bone,Name="spine_05"),Weight=0.2),(Item=(Type=Bone,Name="spine_04"),Weight=0.5))
    Weight=1
   )
   > &RigUnit_ParentConstraint(
    AdvancedSettings=(RotationOrderForFilter=XZY)
    Child=(Type=Bone,Name="CR_Spline_Headphone_Wire_03")
    Parents=((Item=(Type=Bone,Name="spine_02"),Weight=0.05),(Item=(Type=Bone,Name="spine_03"),Weight=0.05),(Item=(Type=Bone,Name="spine_04"),Weight=0.7),(Item=(Type=Bone,Name="dyn_headphones"),Weight=0.2))
    Weight=1
   )
   > &RigUnit_ParentConstraint(
    AdvancedSettings=(RotationOrderForFilter=XZY)
    Child=(Type=Bone,Name="CR_Spline_Headphone_Wire_04")
    Parents="((Item=(Type=Bone,Name=\"def_belt\"),Weight=0.0008),(Item=(Type=Bone,Name=\"pelvis\"),Weight=0.011),(Item=(Type=Bone,Name=\"spine_01\"),Weight=0.021),(Item=(Type=Bone,Name=\"spine_03\"),Weight=0.1058),(Item=(Type=Bone,Name=\"spine_02\"),Weight=0.1532),(Item=(Type=Bo ...(truncated)"
    Weight=1
   )
   > &RigUnit_ParentConstraint(
    AdvancedSettings=(RotationOrderForFilter=XZY)
    Child=(Type=Bone,Name="CR_Spline_Headphone_Wire_05")
    Parents="((Item=(Type=Bone,Name=\"def_belt\"),Weight=0.0048),(Item=(Type=Bone,Name=\"pelvis\"),Weight=0.0258),(Item=(Type=Bone,Name=\"spine_01\"),Weight=0.0305),(Item=(Type=Bone,Name=\"spine_03\"),Weight=0.12),(Item=(Type=Bone,Name=\"spine_02\"),Weight=0.2043),(Item=(Type=Bo ...(truncated)"
    Weight=1
   )
   n3 = RigVMVariableNode() {variable=SplineBones}
   ItemArray = RigVMDispatchNode(Value=((Type=Bone,Name="deform_headphone_wire_01"),(Type=Bone,Name="deform_headphone_wire_02"),(Type=Bone,Name="deform_headphone_wire_03"),(Type=Bone,Name="deform_headphone_wire_04"),(Type=Bone,Name="deform_headphone_wire_05"))) {template="DISPATCH_RigVMDispatch_Constant(io Value)"}
   > /Script/ControlRigSpline.RigUnit_SplineConstraint(
    Alignment=Stretched
    Maximum=1
    Minimum=0
    PrimaryAxis=(X=1,Y=0,Z=0)
    SecondaryAxis=(X=0,Y=0,Z=1)
    bPropagateToChildren=true
    Spline=/Script/ControlRigSpline.RigUnit_ControlRigSplineFromTransforms(
     Compression=0
     SamplesPerSegment=4
     SplineMode=Hermite
     Stretch=1
     bClosed=False
     Transforms=/Script/ControlRig.RigUnit_GetTransformItemArray(Space=GlobalSpace, bInitial=False, Items=@n3.Value) {is=struct RigVMUnitNode}.Transforms
    ) {is=struct RigVMUnitNode}.Spline
    Items=@ItemArray.Value
   ) {is=struct RigVMUnitNode}
   > n4 = &RigUnit_ModifyTransforms(
    ItemToModify="((Item=(Type=Bone,Name=\"deform_headphone_wire_01\"),Transform=(Rotation=(X=0,Y=0,Z=0,W=-1),Translation=(X=0,Y=-0.3666,Z=0),Scale3D=(X=1,Y=1,Z=1))),(Item=(Type=Bone,Name=\"deform_headphone_wire_02\"),Transform=(Rotation=(X=0.0153,Y=0.0021,Z=-0.0057,W=-0.9999), ...(truncated)"
    Weight=1
    WeightMinimum=0
   )
   RigVMFunctionReturnNode(ExecutePin=@n4.ExecutePin)
   RigVMVariableNode() {variable=WireSpline}
  ]}
 CollapseNode_ContainedGraph {nodes [
   RigVMFunctionEntryNode()
   > RigVMFunction_Sequence = RigVMAggregateNode()
   ItemArray = RigVMDispatchNode(Value=((Type=Bone,Name="dyn_skirt_front_l_base"),(Type=Bone,Name="dyn_skirt_side_l_base"))) {template="DISPATCH_RigVMDispatch_Constant(io Value)"}
   For_Each = RigVMDispatchNode(Array=@ItemArray.Value, ExecuteContext=@RigVMFunction_Sequence.A) {template="DISPATCH_RigVMDispatch_ArrayIterator(in Array,out Element,out Index,out Count,out Ratio)"}
   > &RigUnit_ParentConstraint(
    AdvancedSettings=(InterpolationType=Average,RotationOrderForFilter=XZY)
    Parents=((Item=(Type=Bone,Name="cr_thigh_l_no_twist"),Weight=1))
    Weight=1
    Child=@For_Each.Element
   )
   ItemArray_1 = RigVMDispatchNode(Value=((Type=Bone,Name="dyn_skirt_front_r_base"),(Type=Bone,Name="dyn_skirt_side_r_base"))) {template="DISPATCH_RigVMDispatch_Constant(io Value)"}
   For_Each_1 = RigVMDispatchNode(Array=@ItemArray_1.Value, ExecuteContext=@RigVMFunction_Sequence.B) {template="DISPATCH_RigVMDispatch_ArrayIterator(in Array,out Element,out Index,out Count,out Ratio)"}
   > &RigUnit_ParentConstraint(
    AdvancedSettings=(InterpolationType=Average,RotationOrderForFilter=XZY)
    Parents=((Item=(Type=Bone,Name="cr_thigh_r_no_twist"),Weight=1))
    Weight=1
    Child=@For_Each_1.Element
   )
   RigVMFunctionReturnNode(ExecuteContext=@RigVMFunction_Sequence.H)
   Get_Animation_Attribute_3 = RigVMDispatchNode(BoneName=None, Default=0, Name=deformRig_r_thigh_up_alpha) {template="DISPATCH_RigDispatch_GetAnimAttribute(in Name,in BoneName,in Default,out Value,out Found)"}
   n1 = &RigVMFunction_MathDoubleRemap(SourceMaximum=1, SourceMinimum=0, TargetMaximum=0.8, TargetMinimum=0.5, Value=@Get_Animation_Attribute_3.Value)
   Get_Animation_Attribute_2 = RigVMDispatchNode(BoneName=None, Default=0, Name=deformRig_l_thigh_up_alpha) {template="DISPATCH_RigDispatch_GetAnimAttribute(in Name,in BoneName,in Default,out Value,out Found)"}
   Remap = &RigVMFunction_MathDoubleRemap(SourceMaximum=1, SourceMinimum=0, TargetMaximum=0.8, TargetMinimum=0.5, Value=@Get_Animation_Attribute_2.Value)
   Subtract = &RigVMFunction_MathDoubleSub(A=1, B=@Remap.Result)
   Add = &RigVMFunction_MathDoubleAdd(B=@n1.Result, A=@Subtract.Result)
   n2 = /Script/ControlRig.RigUnit_ParentConstraintMath(
    AdvancedSettings=(InterpolationType=Average)
    Parents=((Item=(Type=Bone,Name="dyn_skirt_front_l_base"),Weight=0.8),(Item=(Type=Bone,Name="dyn_skirt_front_r_base"),Weight=0.2))
    Parents.1.Weight=@Add.Result
    Parents.0.Weight=&RigVMFunction_MathDoubleAdd(A=@Remap.Result, B=&RigVMFunction_MathDoubleSub(A=1, B=@n1.Result).Result).Result
    Input=&RigUnit_GetTransform(Item=(Type=Bone,Name="dyn_skirt_front_base"), bInitial=true).Transform
   ) {is=struct RigVMUnitNode}
   n3 = &RigUnit_GetTransform(Item=(Type=Bone,Name="dyn_skirt_front_02"), bInitial=true)
   RigVMFunction_MathTransformTransformVector = /Script/RigVM.RigVMFunction_MathTransformTransformVector(Location=(X=20,Y=0,Z=0), Transform=@n3.Transform) {is=struct RigVMUnitNode}
   Get_Animation_Attribute_1 = RigVMDispatchNode(BoneName=None, Default=0, Name=deformRig_l_thigh_up_alpha) {template="DISPATCH_RigDispatch_GetAnimAttribute(in Name,in BoneName,in Default,out Value,out Found)"}
   Get_Animation_Attribute_6 = RigVMDispatchNode(BoneName=None, Default=0, Name=deformRig_r_thigh_up_alpha) {template="DISPATCH_RigDispatch_GetAnimAttribute(in Name,in BoneName,in Default,out Value,out Found)"}
   n4 = &RigUnit_ParentConstraintMath(
    AdvancedSettings=(InterpolationType=Average)
    Input=(Rotation=(X=0,Y=0,Z=0,W=1),Translation=(X=0,Y=0,Z=0),Scale3D=(X=1,Y=1,Z=1))
    Parents=((Item=(Type=Bone,Name="dyn_skirt_front_l_base"),Weight=0.5),(Item=(Type=Bone,Name="dyn_skirt_front_r_base"),Weight=0.5))
    Input.Rotation=@n3.Transform.Rotation
    Input.Translation=@RigVMFunction_MathTransformTransformVector.Result
    Parents.1.Weight=&RigVMFunction_MathDoubleSub(
     A=1
     B=&RigVMFunction_MathDoubleRemap(SourceMaximum=1, SourceMinimum=0, TargetMaximum=0.8, TargetMinimum=0.5, Value=@Get_Animation_Attribute_1.Value).Result
    ).Result
    Parents.0.Weight=&RigVMFunction_MathDoubleSub(
     A=1
     B=&RigVMFunction_MathDoubleRemap(SourceMaximum=1, SourceMinimum=0, TargetMaximum=0.8, TargetMinimum=0.5, Value=@Get_Animation_Attribute_6.Value).Result
    ).Result
   )
   n5 = /Script/ControlRig.RigUnit_AimBoneMath(
    DebugSettings=(bEnabled=False,Scale=10,WorldOffset=(Rotation=(X=0,Y=0,Z=0,W=1),Translation=(X=0,Y=0,Z=0),Scale3D=(X=1,Y=1,Z=1)))
    Primary=(Weight=1,Axis=(X=1,Y=0,Z=0),Target=(X=1,Y=0,Z=0),Kind=Location,Space=(Type=Bone,Name="None"))
    Secondary=(Weight=1,Axis=(X=0,Y=0,Z=-1),Target=(X=1,Y=0,Z=0),Kind=Location,Space=(Type=Bone,Name="None"))
    Weight=1
    InputTransform=@n2.Output
    Primary.Target=@n4.Output.Translation
    Secondary.Target=&RigUnit_GetTransform(Item=(Type=Bone,Name="dyn_skirt_front_l_base"), bInitial=False).Transform.Translation
   ) {is=struct RigVMUnitNode}
   RigUnit_SetTransform = &RigUnit_SetTransform(
    Item=(Type=Bone,Name="dyn_skirt_front_base")
    Value=(Rotation=(X=0.6519,Y=0.3124,Z=-0.634,W=-0.2747),Translation=(X=0.8505,Y=3.5922,Z=101.7967),Scale3D=(X=1,Y=1,Z=1))
    Weight=1
    bInitial=False
    Value.Rotation=@n5.Result.Rotation
    Value.Translation=@n2.Output.Translation
    ExecutePin=@RigVMFunction_Sequence.C
   )
   n6 = &RigUnit_SphericalPoseReader(
    ActiveRegionScaleFactors=(PositiveWidth=3,NegativeWidth=2,PositiveHeight=3,NegativeHeight=3)
    ActiveRegionSize=0.1
    Debug=(bDrawDebug=false,bDraw2D=False,bDrawLocalAxes=False,DebugScale=25,DebugSegments=20,DebugThickness=0.25)
    DriverAxis=(X=-1,Y=0,Z=0)
    DriverItem=(Type=Bone,Name="thigh_l")
    FalloffRegionScaleFactors=(PositiveWidth=1,NegativeWidth=0.95,PositiveHeight=1,NegativeHeight=1)
    FalloffSize=0.5
    FlipHeightScaling=False
    FlipWidthScaling=False
    OptionalParentItem=(Type=Bone,Name="pelvis")
    RotationOffset=(X=-90,Y=0,Z=0)
    ExecutePin=@RigVMFunction_Sequence.D
   )
   > &RigUnit_ParentConstraint(
    AdvancedSettings=(InterpolationType=Average,RotationOrderForFilter=XZY)
    Child=(Type=Bone,Name="dyn_skirt_rear_l_base")
    Parents=((Item=(Type=Bone,Name="cr_thigh_l_no_twist"),Weight=1))
    Weight=&RigVMFunction_MathDoubleRemap(SourceMaximum=0.3, SourceMinimum=0, TargetMaximum=1, TargetMinimum=0.25, Value=@n6.OutputParam).Result
   )
   > n7 = &RigUnit_SphericalPoseReader(
    ActiveRegionScaleFactors=(PositiveWidth=3,NegativeWidth=2,PositiveHeight=3,NegativeHeight=3)
    ActiveRegionSize=0.1
    Debug=(bDrawDebug=false,bDraw2D=False,bDrawLocalAxes=False,DebugScale=25,DebugSegments=20,DebugThickness=0.25)
    DriverAxis=(X=1,Y=0,Z=0)
    DriverItem=(Type=Bone,Name="thigh_r")
    FalloffRegionScaleFactors=(PositiveWidth=1,NegativeWidth=0.95,PositiveHeight=1,NegativeHeight=1)
    FalloffSize=0.5
    FlipHeightScaling=true
    FlipWidthScaling=true
    OptionalParentItem=(Type=Bone,Name="pelvis")
    RotationOffset=(X=90,Y=0,Z=0)
   )
   > &RigUnit_ParentConstraint(
    AdvancedSettings=(InterpolationType=Average,RotationOrderForFilter=XZY)
    Child=(Type=Bone,Name="dyn_skirt_rear_r_base")
    Parents=((Item=(Type=Bone,Name="cr_thigh_r_no_twist"),Weight=1))
    Weight=&RigVMFunction_MathDoubleRemap(SourceMaximum=0.3, SourceMinimum=0, TargetMaximum=1, TargetMinimum=0.25, Value=@n7.OutputParam).Result
   )
   n8 = &RigUnit_GetTransform(Item=(Type=Bone,Name="dyn_skirt_rear_02"), bInitial=true)
   n9 = &RigUnit_ParentConstraintMath(
    AdvancedSettings=(InterpolationType=Average)
    Input=(Rotation=(X=0,Y=0,Z=0,W=1),Translation=(X=0,Y=0,Z=0),Scale3D=(X=1,Y=1,Z=1))
    Parents=((Item=(Type=Bone,Name="dyn_skirt_rear_l_base"),Weight=0.5),(Item=(Type=Bone,Name="dyn_skirt_rear_r_base"),Weight=0.5))
    Input.Rotation=@n8.Transform.Rotation
    Input.Translation=&RigVMFunction_MathTransformTransformVector(Location=(X=20,Y=0,Z=0), Transform=@n8.Transform).Result
   )
   n10 = &RigUnit_ParentConstraintMath(
    AdvancedSettings=(InterpolationType=Average)
    Parents=((Item=(Type=Bone,Name="dyn_skirt_rear_l_base"),Weight=0.5),(Item=(Type=Bone,Name="dyn_skirt_rear_r_base"),Weight=0.5))
    Input=&RigUnit_GetTransform(Item=(Type=Bone,Name="dyn_skirt_rear_base"), bInitial=true).Transform
   )
   n11 = &RigUnit_AimBoneMath(
    DebugSettings=(bEnabled=False,Scale=10,WorldOffset=(Rotation=(X=0,Y=0,Z=0,W=1),Translation=(X=0,Y=0,Z=0),Scale3D=(X=1,Y=1,Z=1)))
    Primary=(Weight=1,Axis=(X=1,Y=0,Z=0),Target=(X=1,Y=0,Z=0),Kind=Location,Space=(Type=Bone,Name="None"))
    Secondary=(Weight=1,Axis=(X=0,Y=0,Z=1),Target=(X=1,Y=0,Z=0),Kind=Location,Space=(Type=Bone,Name="None"))
    Weight=1
    Primary.Target=@n9.Output.Translation
    InputTransform=@n10.Output
    Secondary.Target=&RigUnit_GetTransform(Item=(Type=Bone,Name="dyn_skirt_rear_l_base"), bInitial=false).Transform.Translation
   )
   > &RigUnit_SetTransform(
    Item=(Type=Bone,Name="dyn_skirt_rear_base")
    Value=(Rotation=(X=-0.1035,Y=0.7055,Z=-0.0476,W=0.6995),Translation=(X=-0.0002,Y=-11.9454,Z=99.7946),Scale3D=(X=1,Y=1,Z=1))
    Weight=1
    bInitial=False
    Value.Rotation=@n11.Result.Rotation
    Value.Translation=@n10.Output.Translation
   )
   > Modify Transforms with Mirror = RigVMFunctionReferenceNode(
    Base Modify Data=((Item=(Type=Bone,Name="dyn_skirt_rear_l_base"),Transform=(Rotation=(X=0,Y=0,Z=-0.0436,W=0.999),Translation=(X=0,Y=-1,Z=0),Scale3D=(X=1,Y=1,Z=1))))
    Mirror Modify Data=((Item=(Type=Bone,Name="dyn_skirt_rear_r_base"),Transform=(Rotation=(X=0,Y=0,Z=0,W=1),Translation=(X=0,Y=0,Z=0),Scale3D=(X=1,Y=1,Z=1))))
    Weight=&RigVMFunction_MathDoubleRemap(SourceMaximum=0.3, SourceMinimum=0, TargetMaximum=1, TargetMinimum=0, Value=RigVMVariableNode() {variable=SPR_Thigh_L_Side}.Value).Result
    "Mirror Weight "=&RigVMFunction_MathDoubleRemap(SourceMaximum=0.3, SourceMinimum=0, TargetMaximum=1, TargetMinimum=0, Value=RigVMVariableNode() {variable=SPR_Thigh_R_Side}.Value).Result
   )
   > Modify Transforms with Mirror_1 = RigVMFunctionReferenceNode(
    Base Modify Data=((Item=(Type=Bone,Name="dyn_skirt_rear_l_base"),Transform=(Rotation=(X=0.0872,Y=0,Z=0,W=0.9962),Translation=(X=2,Y=1,Z=-2),Scale3D=(X=1,Y=1,Z=1))))
    Mirror Modify Data=((Item=(Type=Bone,Name="dyn_skirt_rear_r_base"),Transform=(Rotation=(X=0,Y=0,Z=0,W=1),Translation=(X=0,Y=0,Z=0),Scale3D=(X=1,Y=1,Z=1))))
    Weight=RigVMVariableNode() {variable=SPR_Thigh_L_Back}.Value
    "Mirror Weight "=RigVMVariableNode() {variable=SPR_Thigh_R_Back}.Value
   )
   Multiply = &RigVMFunction_MathDoubleMul(
    B=0.5
    A=&RigVMFunction_MathDoubleAdd(A=RigVMVariableNode() {variable=SPR_Thigh_L_Back}.Value, B=RigVMVariableNode() {variable=SPR_Thigh_R_Back}.Value).Result
   )
   > &RigUnit_ModifyTransforms(
    ItemToModify=((Item=(Type=Bone,Name="dyn_skirt_rear_base"),Transform=(Rotation=(X=0.1113,Y=0.0284,Z=-0.3733,W=0.9206),Translation=(X=4.317,Y=-1.5324,Z=-0.2195),Scale3D=(X=1,Y=1,Z=1))))
    WeightMinimum=0
    Weight=@Multiply.Result
   )
   Get_Animation_Attribute_4 = RigVMDispatchNode(BoneName=None, Default=0, Name=deformRig_l_thigh_up_alpha) {template="DISPATCH_RigDispatch_GetAnimAttribute(in Name,in BoneName,in Default,out Value,out Found)"}
   &RigUnit_ModifyTransforms(
    ItemToModify="((Item=(Type=Bone,Name=\"dyn_skirt_rear_l_base\"),Transform=(Rotation=(X=0,Y=0,Z=0,W=1),Translation=(X=0,Y=-2,Z=0),Scale3D=(X=1,Y=1,Z=1))),(Item=(Type=Bone,Name=\"dyn_skirt_rear_base\"),Transform=(Rotation=(X=0,Y=0,Z=0,W=1),Translation=(X=0,Y=-1,Z=0),Scale3D=( ...(truncated)"
    WeightMinimum=0
    Weight=@Get_Animation_Attribute_4.Value
    ExecutePin=@RigVMFunction_Sequence.E
   )
   Get_Animation_Attribute_5 = RigVMDispatchNode(BoneName=None, Default=0, Name=deformRig_r_thigh_up_alpha) {template="DISPATCH_RigDispatch_GetAnimAttribute(in Name,in BoneName,in Default,out Value,out Found)"}
   > &RigUnit_ModifyTransforms(
    ItemToModify="((Item=(Type=Bone,Name=\"dyn_skirt_rear_r_base\"),Transform=(Rotation=(X=0,Y=0,Z=0,W=1),Translation=(X=0,Y=2,Z=0),Scale3D=(X=1,Y=1,Z=1))),(Item=(Type=Bone,Name=\"dyn_skirt_rear_base\"),Transform=(Rotation=(X=0,Y=0,Z=0,W=1),Translation=(X=0,Y=-1,Z=0),Scale3D=(X ...(truncated)"
    WeightMinimum=0
    Weight=@Get_Animation_Attribute_5.Value
   )
   &RigUnit_ModifyTransforms(
    ItemToModify="((Item=(Type=Bone,Name=\"dyn_skirt_front_02\"),Transform=(Rotation=(X=0,Y=0,Z=0,W=1),Translation=(X=0,Y=-0.5,Z=0),Scale3D=(X=1,Y=1,Z=1))),(Item=(Type=Bone,Name=\"dyn_skirt_front_l_02\"),Transform=(Rotation=(X=0,Y=0,Z=0,W=1),Translation=(X=0,Y=-1,Z=0),Scale3D=( ...(truncated)"
    Weight=1
    WeightMinimum=0
    ExecutePin=@RigVMFunction_Sequence.G
   )
   &RigUnit_SphericalPoseReader(
    ActiveRegionScaleFactors=(PositiveWidth=3,NegativeWidth=1,PositiveHeight=1,NegativeHeight=1)
    ActiveRegionSize=0.1
    Debug=(bDrawDebug=false,bDraw2D=False,bDrawLocalAxes=False,DebugScale=25,DebugSegments=20,DebugThickness=0.25)
    DriverAxis=(X=-1,Y=0,Z=0)
    DriverItem=(Type=Bone,Name="thigh_l")
    ExecutePin=()
    FalloffRegionScaleFactors=(PositiveWidth=1,NegativeWidth=1,PositiveHeight=1,NegativeHeight=1)
    FalloffSize=0.5
    FlipHeightScaling=False
    FlipWidthScaling=False
    OptionalParentItem=(Type=Bone,Name="None")
    RotationOffset=(X=180,Y=0,Z=0)
   )
   > &RigUnit_SphericalPoseReader(
    ActiveRegionScaleFactors=(PositiveWidth=3,NegativeWidth=1,PositiveHeight=1,NegativeHeight=1)
    ActiveRegionSize=0.1
    Debug=(bDrawDebug=false,bDraw2D=False,bDrawLocalAxes=False,DebugScale=25,DebugSegments=20,DebugThickness=0.25)
    DriverAxis=(X=1,Y=0,Z=0)
    DriverItem=(Type=Bone,Name="thigh_r")
    FalloffRegionScaleFactors=(PositiveWidth=1,NegativeWidth=1,PositiveHeight=1,NegativeHeight=1)
    FalloffSize=0.5
    FlipHeightScaling=true
    FlipWidthScaling=true
    OptionalParentItem=(Type=Bone,Name="None")
    RotationOffset=(X=0,Y=0,Z=0)
   )
  ]}
 CollapseNode_ContainedGraph {nodes [
   n1 = RigVMFunctionEntryNode()
   Modify Transforms with Mirror_4 = RigVMFunctionReferenceNode(
    Base Modify Data="((Item=(Type=Bone,Name=\"dyn_jacket_front_l_base\"),Transform=(Rotation=(X=0,Y=-0.0175,Z=0,W=0.9998),Translation=(X=0,Y=-3,Z=0),Scale3D=(X=1,Y=1,Z=1))),(Item=(Type=Bone,Name=\"dyn_jacket_rear_l_base\"),Transform=(Rotation=(X=0,Y=-0.0175,Z=0,W=0.9998),Translati ...(truncated)"
    Mirror Modify Data="((Item=(Type=Bone,Name=\"dyn_jacket_front_r_base\"),Transform=(Rotation=(X=0,Y=0,Z=0,W=1),Translation=(X=0,Y=0,Z=0),Scale3D=(X=1,Y=1,Z=1))),(Item=(Type=Bone,Name=\"dyn_jacket_rear_r_base\"),Transform=(Rotation=(X=0,Y=0,Z=0,W=1),Translation=(X=0,Y=0,Z=0),Scale3 ...(truncated)"
    "Mirror Weight "=1
    Weight=1
    ExecuteContext=@n1.ExecutePin
   )
   > &RigUnit_ModifyTransforms(
    ItemToModify="((Item=(Type=Bone,Name=\"dyn_jacket_rear_base\"),Transform=(Rotation=(X=0,Y=0.0175,Z=0,W=0.9998),Translation=(X=0,Y=-1.5,Z=0),Scale3D=(X=1,Y=1,Z=1))),(Item=(Type=Bone,Name=\"dyn_skirt_side_l_base\"),Transform=(Rotation=(X=0,Y=0,Z=0,W=1),Translation=(X=0,Y=0.5, ...(truncated)"
    Weight=1
    WeightMinimum=0
   )
   > RigVMFunction_Sequence = RigVMAggregateNode()
   RigVMFunctionReturnNode(ExecutePin=@RigVMFunction_Sequence.C)
   KalmanFilter = &RigVMFunction_KalmanFloat(
    BufferSize=4
    Value=RigVMDispatchNode(False=0, True=1, Condition=RigVMVariableNode() {variable=IsKneeSliding}.Value) {template="DISPATCH_RigVMDispatch_If(in Condition,in True,in False,out Result)"}.Result
   ) {template="KalmanFilter::Execute(in Value,in BufferSize,out Result)"}
   &RigUnit_ModifyTransforms(
    ItemToModify=((Item=(Type=Bone,Name="None"),Transform=(Rotation=(X=0,Y=0,Z=0,W=1),Translation=(X=0,Y=0,Z=0),Scale3D=(X=1,Y=1,Z=1))))
    WeightMinimum=0
    Weight=@KalmanFilter.Result
    ExecutePin=@RigVMFunction_Sequence.B
   )
   n2 = RigVMVariableNode() {variable=IsBaseballSliding}
   > &RigUnit_ModifyTransforms(
    ItemToModify=((Item=(Type=Bone,Name="None"),Transform=(Rotation=(X=0,Y=0,Z=0,W=1),Translation=(X=0,Y=0,Z=0),Scale3D=(X=1,Y=1,Z=1))))
    WeightMinimum=0
    Weight=&RigVMFunction_KalmanFloat(
     BufferSize=4
     Value=RigVMDispatchNode(False=0, True=1, Condition=@n2.Value) {template="DISPATCH_RigVMDispatch_If(in Condition,in True,in False,out Result)"}.Result
    ) {template="KalmanFilter::Execute(in Value,in BufferSize,out Result)"}.Result
   )
   Modify Transforms with Mirror = RigVMFunctionReferenceNode(
    Base Modify Data=((Item=(Type=Bone,Name="dyn_jacket_front_l_base"),Transform=(Rotation=(X=0.3657,Y=0.1307,Z=-0.1421,W=0.9105),Translation=(X=0,Y=0,Z=0),Scale3D=(X=1,Y=1,Z=1))))
    Mirror Modify Data=((Item=(Type=Bone,Name="dyn_jacket_front_r_base"),Transform=(Rotation=(X=0,Y=0,Z=0,W=1),Translation=(X=0,Y=0,Z=0),Scale3D=(X=1,Y=1,Z=1))))
    Weight=RigVMVariableNode() {variable=SPR_Thigh_L_Up}.Value
    "Mirror Weight "=RigVMVariableNode() {variable=SPR_Thigh_R_Up}.Value
    ExecuteContext=@RigVMFunction_Sequence.A
   )
   > Modify Transforms with Mirror_3 = RigVMFunctionReferenceNode(
    Base Modify Data="((Item=(Type=Bone,Name=\"dyn_jacket_front_l_base\"),Transform=(Rotation=(X=0,Y=0,Z=0,W=1),Translation=(X=0,Y=0,Z=0),Scale3D=(X=1,Y=1,Z=1))),(Item=(Type=Bone,Name=\"dyn_jacket_rear_l_base\"),Transform=(Rotation=(X=-0.0826,Y=0.3016,Z=0.1163,W=-0.9427),Translatio ...(truncated)"
    Mirror Modify Data="((Item=(Type=Bone,Name=\"dyn_jacket_front_r_base\"),Transform=(Rotation=(X=0,Y=0,Z=0,W=1),Translation=(X=0,Y=0,Z=0),Scale3D=(X=1,Y=1,Z=1))),(Item=(Type=Bone,Name=\"dyn_jacket_rear_r_base\"),Transform=(Rotation=(X=0,Y=0,Z=0,W=1),Translation=(X=0,Y=0,Z=0),Scale3 ...(truncated)"
    Weight=RigVMVariableNode() {variable=SPR_Thigh_L_Inbetween_Up}.Value
    "Mirror Weight "=RigVMVariableNode() {variable=SPR_Thigh_R_Inbetween_Up}.Value
   )
   RigVMFunction_AnimEvalRichCurve = &RigVMFunction_AnimEvalRichCurve(
    Curve=(EditorCurveData=(Keys=((),(Time=0.3695,Value=0.1048),(Time=1,Value=1))))
    SourceMinimum=0
    TargetMinimum=0
    Value=RigVMVariableNode() {variable=SPR_Thigh_R_Side}.Value
   )
   n3 = &RigVMFunction_AnimEvalRichCurve(
    Curve=(EditorCurveData=(Keys=((),(Time=0.3695,Value=0.1561),(Time=1,Value=1))))
    SourceMinimum=0
    TargetMinimum=0
    Value=RigVMVariableNode() {variable=SPR_Thigh_L_Side}.Value
   )
   > Modify Transforms with Mirror_1 = RigVMFunctionReferenceNode(
    Base Modify Data="((Item=(Type=Bone,Name=\"dyn_jacket_front_l_base\"),Transform=(Rotation=(X=0.3375,Y=0.1454,Z=-0.5739,W=0.7319),Translation=(X=9.4591,Y=4.5545,Z=-2.1571),Scale3D=(X=1,Y=1,Z=1))),(Item=(Type=Bone,Name=\"dyn_jacket_rear_l_base\"),Transform=(Rotation=(X=-0.1037,Y= ...(truncated)"
    Mirror Modify Data="((Item=(Type=Bone,Name=\"dyn_jacket_front_r_base\"),Transform=(Rotation=(X=0,Y=0,Z=0,W=1),Translation=(X=0,Y=0,Z=0),Scale3D=(X=1,Y=1,Z=1))),(Item=(Type=Bone,Name=\"dyn_jacket_rear_r_base\"),Transform=(Rotation=(X=0,Y=0,Z=0,W=1),Translation=(X=0,Y=0,Z=0),Scale3 ...(truncated)"
    "Mirror Weight "=@RigVMFunction_AnimEvalRichCurve.Result
    Weight=@n3.Result
   )
   > Modify Transforms with Mirror_2 = RigVMFunctionReferenceNode(
    Base Modify Data="((Item=(Type=Bone,Name=\"dyn_jacket_front_l_base\"),Transform=(Rotation=(X=-0.2501,Y=0.3688,Z=0.1915,W=0.8745),Translation=(X=0,Y=0,Z=0),Scale3D=(X=1,Y=1,Z=1))),(Item=(Type=Bone,Name=\"dyn_jacket_rear_l_base\"),Transform=(Rotation=(X=0.1045,Y=0.2058,Z=-0.5978, ...(truncated)"
    Mirror Modify Data="((Item=(Type=Bone,Name=\"dyn_jacket_front_r_base\"),Transform=(Rotation=(X=0,Y=0,Z=0,W=1),Translation=(X=0,Y=0,Z=0),Scale3D=(X=1,Y=1,Z=1))),(Item=(Type=Bone,Name=\"dyn_jacket_rear_r_base\"),Transform=(Rotation=(X=0,Y=0,Z=0,W=1),Translation=(X=0,Y=0,Z=0),Scale3 ...(truncated)"
    Weight=RigVMVariableNode() {variable=SPR_Thigh_L_Back}.Value
    "Mirror Weight "=RigVMVariableNode() {variable=SPR_Thigh_R_Back}.Value
   )
  ]}
 CollapseNode_ContainedGraph {nodes [
   n1 = RigVMFunctionEntryNode()
   n2 = &RigUnit_SphericalPoseReader(
    ActiveRegionScaleFactors=(PositiveWidth=1,NegativeWidth=1,PositiveHeight=1,NegativeHeight=1)
    ActiveRegionSize=0.1
    Debug=(bDrawDebug=True,bDraw2D=False,bDrawLocalAxes=False,DebugScale=25,DebugSegments=20,DebugThickness=0.25)
    DriverAxis=(X=1,Y=0,Z=0)
    DriverItem=(Type=Bone,Name="clavicle_l")
    FalloffRegionScaleFactors=(PositiveWidth=1,NegativeWidth=1,PositiveHeight=1,NegativeHeight=1)
    FalloffSize=0.2
    FlipHeightScaling=False
    FlipWidthScaling=False
    OptionalParentItem=(Type=Bone,Name="None")
    RotationOffset=(X=56.0078,Y=11.7977,Z=-71.4778)
    ExecutePin=@n1.ExecutePin
   )
   > n3 = &RigUnit_SphericalPoseReader(
    ActiveRegionScaleFactors=(PositiveWidth=1,NegativeWidth=1,PositiveHeight=1,NegativeHeight=1)
    ActiveRegionSize=0.1
    Debug=(bDrawDebug=true,bDraw2D=False,bDrawLocalAxes=False,DebugScale=25,DebugSegments=20,DebugThickness=0.25)
    DriverAxis=(X=-1,Y=0,Z=0)
    DriverItem=(Type=Bone,Name="clavicle_r")
    FalloffRegionScaleFactors=(PositiveWidth=1,NegativeWidth=1,PositiveHeight=1,NegativeHeight=1)
    FalloffSize=0.2
    FlipHeightScaling=False
    FlipWidthScaling=False
    OptionalParentItem=(Type=Bone,Name="None")
    RotationOffset=(X=123.3708,Y=3.288,Z=109.474)
   )
   > &RigUnit_ModifyTransforms(
    ItemToModify=((Item=(Type=Bone,Name="dyn_lapel_l_base"),Transform=(Rotation=(X=-0.2274,Y=0.1343,Z=-0.1185,W=0.9572),Translation=(X=-2.4459,Y=-3.0086,Z=1.5363),Scale3D=(X=1,Y=1,Z=1))))
    WeightMinimum=0
    Weight=@n2.OutputParam
   )
   > n4 = &RigUnit_ModifyTransforms(
    ItemToModify=((Item=(Type=Bone,Name="dyn_lapel_r_base"),Transform=(Rotation=(X=0.0487,Y=0.057,Z=-0.0896,W=0.9932),Translation=(X=1.3531,Y=-1.3214,Z=0.7861),Scale3D=(X=1,Y=1,Z=1))))
    WeightMinimum=0
    Weight=@n3.OutputParam
   )
   RigVMFunctionReturnNode(ExecutePin=@n4.ExecutePin)
   Modify Transforms with Mirror = RigVMFunctionReferenceNode(
    Base Modify Data=((Item=(Type=Bone,Name="dyn_lapel_l_base"),Transform=(Rotation=(X=-0.2274,Y=0.1343,Z=-0.1185,W=0.9572),Translation=(X=-2.4459,Y=-3.0086,Z=1.5363),Scale3D=(X=1,Y=1,Z=1))))
    ExecuteContext=()
    Mirror Modify Data=((Item=(Type=Bone,Name="dyn_lapel_r_base"),Transform=(Rotation=(X=0,Y=0,Z=0,W=1),Translation=(X=0,Y=0,Z=0),Scale3D=(X=1,Y=1,Z=1))))
    Weight=@n2.OutputParam
    "Mirror Weight "=@n3.OutputParam
   )
  ]}
 CollapseNode_ContainedGraph {nodes [
   n1 = RigVMFunctionEntryNode()
   &RigUnit_RotationConstraintLocalSpaceOffset(
    AdvancedSettings=(InterpolationType=Average,RotationOrderForFilter=XZY)
    Child=(Type=Bone,Name="def_simspace")
    Filter=(bX=false,bY=false,bZ=True)
    Parents=((Item=(Type=Bone,Name="pelvis"),Weight=1))
    Weight=0.5
    ExecutePin=@n1.ExecutePin
   )
   > &RigUnit_PositionConstraintLocalSpaceOffset(
    Child=(Type=Bone,Name="def_simspace")
    Filter=(bX=false,bY=false,bZ=True)
    Parents=((Item=(Type=Bone,Name="pelvis"),Weight=1))
    Weight=0.5
    bMaintainOffset=True
   )
   RigUnit_GetTransform = &RigUnit_GetTransform(Item=(Type=Bone,Name="def_simspace"), bInitial=False)
   > RigVMFunction_DebugTransformMutableNoSpace = /Script/RigVM.RigVMFunction_DebugTransformMutableNoSpace(
    Color=(R=1,G=0,B=0,A=1)
    DebugDrawSettings=(DepthPriority=SDPG_Foreground,Lifetime=-1)
    Mode=Axes
    Scale=10
    Thickness=0
    WorldOffset=(Rotation=(X=0,Y=0,Z=0,W=1),Translation=(X=0,Y=0,Z=0),Scale3D=(X=1,Y=1,Z=1))
    bEnabled=True
    Transform=@RigUnit_GetTransform.Transform
   ) {is=struct RigVMUnitNode}
   RigVMFunctionReturnNode(ExecutePin=@RigVMFunction_DebugTransformMutableNoSpace.ExecutePin)
   RigVMFunction_MathQuaternionSwingTwist = &RigVMFunction_MathQuaternionSwingTwist(
    TwistAxis=(X=0,Y=0,Z=1)
    Input=/Script/RigVM.RigVMFunction_MathQuaternionMul(B=(X=0,Y=0.7071,Z=0,W=0.7071), A=&RigUnit_GetTransform(Item=(Type=Bone,Name="pelvis"), bInitial=False).Transform.Rotation) {is=struct RigVMUnitNode, template="Multiply::Execute(in A,in B,out Result)"}.Result
   )
   RigUnit_SetRelativeTransformForItem = /Script/ControlRig.RigUnit_SetRelativeTransformForItem(
    Child=(Type=Bone,Name="def_simspace")
    ExecutePin=()
    Parent=(Type=Bone,Name="root")
    Value=(Rotation=(X=0,Y=0,Z=0,W=1),Translation=(X=0,Y=0,Z=0),Scale3D=(X=1,Y=1,Z=1))
    Weight=1
    bParentInitial=False
    bPropagateToChildren=True
    Value.Rotation=@RigVMFunction_MathQuaternionSwingTwist.Twist
   ) {
    is=struct RigVMUnitNode
    template="Set Relative Transform::Execute(in Child,in Parent,in bParentInitial,in Value,in Weight,in bPropagateToChildren)"
   }
  ]}
 CollapseNode_ContainedGraph {nodes [
   n1 = RigVMFunctionEntryNode()
   n2 = &RigUnit_SphericalPoseReader(
    ActiveRegionScaleFactors=(PositiveWidth=3,NegativeWidth=1,PositiveHeight=2.4,NegativeHeight=1)
    ActiveRegionSize=0.1
    Debug=(bDrawDebug=false,bDraw2D=False,bDrawLocalAxes=False,DebugScale=25,DebugSegments=20,DebugThickness=0.25)
    DriverAxis=(X=-1,Y=0,Z=0)
    DriverItem=(Type=Bone,Name="thigh_l")
    FalloffRegionScaleFactors=(PositiveWidth=1,NegativeWidth=1,PositiveHeight=0.75,NegativeHeight=0.75)
    FalloffSize=0.58
    FlipHeightScaling=False
    FlipWidthScaling=False
    OptionalParentItem=(Type=Bone,Name="spine_04")
    RotationOffset=(X=90,Y=0,Z=-15)
    ExecutePin=@n1.ExecutePin
   )
   > RigVMVariableNode(Value=@n2.OutputParam) {variable=SPR_Thigh_L_Up}
   > n3 = &RigUnit_SphericalPoseReader(
    ActiveRegionScaleFactors=(PositiveWidth=3,NegativeWidth=1,PositiveHeight=2.4,NegativeHeight=1)
    ActiveRegionSize=0.1
    Debug=(bDrawDebug=false,bDraw2D=False,bDrawLocalAxes=False,DebugScale=25,DebugSegments=20,DebugThickness=0.25)
    DriverAxis=(X=1,Y=0,Z=0)
    DriverItem=(Type=Bone,Name="thigh_r")
    FalloffRegionScaleFactors=(PositiveWidth=1,NegativeWidth=1,PositiveHeight=0.75,NegativeHeight=0.75)
    FalloffSize=0.58
    FlipHeightScaling=true
    FlipWidthScaling=true
    OptionalParentItem=(Type=Bone,Name="spine_04")
    RotationOffset=(X=-90,Y=0,Z=-15)
   )
   > RigVMVariableNode(Value=@n3.OutputParam) {variable=SPR_Thigh_R_Up}
   > n4 = &RigUnit_SphericalPoseReader(
    ActiveRegionScaleFactors=(PositiveWidth=3,NegativeWidth=1,PositiveHeight=1.3,NegativeHeight=1.2)
    ActiveRegionSize=0.1
    Debug=(bDrawDebug=false,bDraw2D=False,bDrawLocalAxes=False,DebugScale=25,DebugSegments=20,DebugThickness=0.25)
    DriverAxis=(X=-1,Y=0,Z=0)
    DriverItem=(Type=Bone,Name="thigh_l")
    FalloffRegionScaleFactors=(PositiveWidth=1,NegativeWidth=1,PositiveHeight=0.5,NegativeHeight=0.5)
    FalloffSize=0.57
    FlipHeightScaling=False
    FlipWidthScaling=False
    OptionalParentItem=(Type=Bone,Name="spine_04")
    RotationOffset=(X=130,Y=5,Z=-12)
   )
   > RigVMVariableNode(Value=@n4.OutputParam) {variable=SPR_Thigh_L_Inbetween_Up}
   > n5 = &RigUnit_SphericalPoseReader(
    ActiveRegionScaleFactors=(PositiveWidth=3,NegativeWidth=1,PositiveHeight=1.3,NegativeHeight=1.2)
    ActiveRegionSize=0.1
    Debug=(bDrawDebug=false,bDraw2D=False,bDrawLocalAxes=False,DebugScale=25,DebugSegments=20,DebugThickness=0.25)
    DriverAxis=(X=1,Y=0,Z=0)
    DriverItem=(Type=Bone,Name="thigh_r")
    FalloffRegionScaleFactors=(PositiveWidth=1,NegativeWidth=1,PositiveHeight=0.5,NegativeHeight=0.5)
    FalloffSize=0.57
    FlipHeightScaling=true
    FlipWidthScaling=true
    OptionalParentItem=(Type=Bone,Name="spine_04")
    RotationOffset=(X=-50,Y=5,Z=-12)
   )
   > RigVMVariableNode(Value=@n5.OutputParam) {variable=SPR_Thigh_R_Inbetween_Up}
   > n6 = &RigUnit_SphericalPoseReader(
    ActiveRegionScaleFactors=(PositiveWidth=3,NegativeWidth=1,PositiveHeight=1,NegativeHeight=1)
    ActiveRegionSize=0.1
    Debug=(bDrawDebug=true,bDraw2D=False,bDrawLocalAxes=False,DebugScale=25,DebugSegments=20,DebugThickness=0.25)
    DriverAxis=(X=-1,Y=0,Z=0)
    DriverItem=(Type=Bone,Name="thigh_l")
    FalloffRegionScaleFactors=(PositiveWidth=1,NegativeWidth=1,PositiveHeight=1,NegativeHeight=1)
    FalloffSize=0.5
    FlipHeightScaling=False
    FlipWidthScaling=False
    OptionalParentItem=(Type=Bone,Name="None")
    RotationOffset=(X=180,Y=0,Z=0)
   )
   > RigVMVariableNode(Value=@n6.OutputParam) {variable=SPR_Thigh_L_Side}
   > n7 = &RigUnit_SphericalPoseReader(
    ActiveRegionScaleFactors=(PositiveWidth=3,NegativeWidth=1,PositiveHeight=1,NegativeHeight=1)
    ActiveRegionSize=0.1
    Debug=(bDrawDebug=true,bDraw2D=False,bDrawLocalAxes=False,DebugScale=25,DebugSegments=20,DebugThickness=0.25)
    DriverAxis=(X=1,Y=0,Z=0)
    DriverItem=(Type=Bone,Name="thigh_r")
    FalloffRegionScaleFactors=(PositiveWidth=1,NegativeWidth=1,PositiveHeight=1,NegativeHeight=1)
    FalloffSize=0.5
    FlipHeightScaling=true
    FlipWidthScaling=true
    OptionalParentItem=(Type=Bone,Name="None")
    RotationOffset=(X=0,Y=0,Z=0)
   )
   > RigVMVariableNode(Value=@n7.OutputParam) {variable=SPR_Thigh_R_Side}
   > n8 = &RigUnit_SphericalPoseReader(
    ActiveRegionScaleFactors=(PositiveWidth=3,NegativeWidth=1,PositiveHeight=3,NegativeHeight=3)
    ActiveRegionSize=0.1
    Debug=(bDrawDebug=false,bDraw2D=False,bDrawLocalAxes=False,DebugScale=25,DebugSegments=20,DebugThickness=0.25)
    DriverAxis=(X=-1,Y=0,Z=0)
    DriverItem=(Type=Bone,Name="thigh_l")
    FalloffRegionScaleFactors=(PositiveWidth=1,NegativeWidth=0.95,PositiveHeight=1,NegativeHeight=1)
    FalloffSize=0.5
    FlipHeightScaling=False
    FlipWidthScaling=False
    OptionalParentItem=(Type=Bone,Name="None")
    RotationOffset=(X=-90,Y=0,Z=0)
   )
   > RigVMVariableNode(Value=@n8.OutputParam) {variable=SPR_Thigh_L_Back}
   > n9 = &RigUnit_SphericalPoseReader(
    ActiveRegionScaleFactors=(PositiveWidth=3,NegativeWidth=1,PositiveHeight=3,NegativeHeight=3)
    ActiveRegionSize=0.1
    Debug=(bDrawDebug=false,bDraw2D=False,bDrawLocalAxes=False,DebugScale=25,DebugSegments=20,DebugThickness=0.25)
    DriverAxis=(X=1,Y=0,Z=0)
    DriverItem=(Type=Bone,Name="thigh_r")
    FalloffRegionScaleFactors=(PositiveWidth=1,NegativeWidth=0.95,PositiveHeight=1,NegativeHeight=1)
    FalloffSize=0.5
    FlipHeightScaling=true
    FlipWidthScaling=true
    OptionalParentItem=(Type=Bone,Name="None")
    RotationOffset=(X=90,Y=0,Z=0)
   )
   > n10 = RigVMVariableNode(Value=@n9.OutputParam) {variable=SPR_Thigh_R_Back}
   RigVMFunctionReturnNode(ExecutePin=@n10.ExecuteContext)
  ]}
 RigVMFunctionLibrary {links [], nodes []}
 RigVMFunction_Sequence_1_ContainedGraph {nodes [
   n1 = RigVMFunctionEntryNode()
   RigVMFunction_Sequence = &RigVMFunction_Sequence(ExecuteContext=@n1.ExecuteContext)
   n2 = &RigVMFunction_Sequence(ExecuteContext=@RigVMFunction_Sequence.B)
   n3 = &RigVMFunction_Sequence(ExecuteContext=@n2.B)
   n4 = &RigVMFunction_Sequence(ExecuteContext=@n3.B)
   n5 = &RigVMFunction_Sequence(ExecuteContext=@n4.B)
   n6 = &RigVMFunction_Sequence(ExecuteContext=@n5.B)
   n7 = &RigVMFunction_Sequence(ExecuteContext=@n6.B)
   RigVMFunctionReturnNode(A=@RigVMFunction_Sequence.A, B=@n2.A, C=@n3.A, D=@n4.A, E=@n5.A, F=@n6.A, G=@n7.A, H=@n7.B)
  ]}
 RigVMFunction_Sequence_1_ContainedGraph {nodes [
   n1 = RigVMFunctionEntryNode()
   RigVMFunction_Sequence = &RigVMFunction_Sequence(ExecuteContext=@n1.ExecuteContext)
   n2 = &RigVMFunction_Sequence(ExecuteContext=@RigVMFunction_Sequence.B)
   RigVMFunctionReturnNode(A=@RigVMFunction_Sequence.A, B=@n2.A, C=@n2.B)
  ]}
 RigVMModel {nodes [
   RigUnit_BeginExecution = &RigUnit_BeginExecution()
   n1 = /Script/FortniteGame.RigUnit_FortDeformRig(
    Bool=(bUpperArm=True,bElbow=True,bWrist=True,bGroin=True,bGlute=True,bKnee=True,bPec=False,bBicep=False,bLat=False)
    Curve="(bLUpperArmUpAlpha=False,bRUpperArmUpAlpha=False,bLUpperArmFrAlpha=False,bRUpperArmFrAlpha=False,bLUpperArmBkAlpha=False,bRUpperArmBkAlpha=False,bLWristUpAlpha=False,bRWristUpAlpha=False,bLWristDnAlpha=False,bRWristDnAlpha=False,bLKneeAlpha=False,bRKneeAlp ...(truncated)"
    Debug=(bUpperArmDebug=False,bWristDebug=False,bLatDebug=False,bGroinGluteDebug=False)
    Value="(UpperArmUp=(Rotation=(X=0.0,Y=-0.0,Z=0.0,W=1.0),Translation=(X=0.0,Y=0.0,Z=8.0),Scale3D=(X=0.70,Y=1.0,Z=1.0)),UpperArmFwdFr=(Rotation=(X=0.0,Y=-0.0,Z=0.0,W=1.0),Translation=(X=0.0,Y=10.0,Z=0.0),Scale3D=(X=0.50,Y=1.0,Z=1.0)),UpperArmFwdBk=(Rotation=(X=0.0, ...(truncated)"
    ExecutePin=@RigUnit_BeginExecution.ExecutePin
   ) {is=struct RigVMUnitNode}
   SPR Cache = RigVMCollapseNode(ExecutePin=@n1.ExecutePin)
   Set Simspace = RigVMCollapseNode(ExecutePin=@"SPR Cache".ExecutePin)
   Deform Armpits = RigVMCollapseNode(ExecutePin=@"Set Simspace".ExecutePin)
   Deform Lapels = RigVMCollapseNode(ExecutePin=@"Deform Armpits".ExecutePin)
   Deform Earphone Wires = RigVMCollapseNode(ExecutePin=@"Deform Lapels".ExecutePin)
   RigUnit_GetTransform = &RigUnit_GetTransform(Item=(Type=Bone,Name="thigh_l"), Space=LocalSpace, bInitial=False)
   n2 = &RigUnit_GetTransform(Item=(Type=Bone,Name="thigh_l"), Space=LocalSpace, bInitial=true)
   Inverse = /Script/RigVM.RigVMFunction_MathQuaternionInverse(Value=&RigVMFunction_MathQuaternionSwingTwist(
     TwistAxis=(X=1,Y=0,Z=0)
     Input=&RigVMFunction_MathTransformMakeRelative(Global=@RigUnit_GetTransform.Transform, Parent=@n2.Transform).Local.Rotation
    ).Twist) {is=struct RigVMUnitNode, template="Inverse::Execute(in Value,out Result)"}
   RigUnit_SetRotation = &RigUnit_SetRotation(Item=(Type=Bone,Name="cr_thigh_l_no_twist"), Weight=1, bInitial=False, ExecutePin=@"Deform Earphone Wires".ExecutePin, Value=@Inverse.Result)
   n3 = &RigVMFunction_MathTransformMakeRelative(
    Global=&RigUnit_GetTransform(Item=(Type=Bone,Name="thigh_r"), Space=LocalSpace, bInitial=False).Transform
    Parent=&RigUnit_GetTransform(Item=(Type=Bone,Name="thigh_r"), Space=LocalSpace, bInitial=true).Transform
   )
   > &RigUnit_SetRotation(
    Item=(Type=Bone,Name="cr_thigh_r_no_twist")
    Weight=1
    bInitial=False
    Value=&RigVMFunction_MathQuaternionInverse(Value=&RigVMFunction_MathQuaternionSwingTwist(TwistAxis=(X=1,Y=0,Z=0), Input=@n3.Local.Rotation).Twist) {template="Inverse::Execute(in Value,out Result)"}.Result
   )
   > Skirt Deformation = RigVMCollapseNode()
   RigVMFunction_ControlFlowBranch = &RigVMFunction_ControlFlowBranch(ExecuteContext=@"Skirt Deformation".ExecuteContext, Condition=RigVMVariableNode() {variable=ClothEnabled}.Value)
   Jacket Deformation = RigVMCollapseNode(ExecutePin=@RigVMFunction_ControlFlowBranch.False)
   &RigUnit_ModifyTransforms(
    ExecutePin=()
    ItemToModify="((Item=(Type=Bone,Name=\"deform_armpit_l_base\"),Transform=(Rotation=(X=0,Y=0,Z=0,W=1),Translation=(X=5,Y=0,Z=0),Scale3D=(X=1,Y=1,Z=1))),(Item=(Type=Bone,Name=\"deform_armpit_r_base\"),Transform=(Rotation=(X=0,Y=0,Z=0,W=1),Translation=(X=-5,Y=0,Z=0),Scale3D=(X ...(truncated)"
    Weight=1
    WeightMinimum=0
   )
  ]}
 RigVMModel Rig {nodes [
   n1 = &RigUnit_PrepareForExecution()
   ItemArray = RigVMDispatchNode(Value=((Type=Bone,Name="deform_headphone_wire_01"),(Type=Bone,Name="deform_headphone_wire_02"),(Type=Bone,Name="deform_headphone_wire_03"),(Type=Bone,Name="deform_headphone_wire_04"),(Type=Bone,Name="deform_headphone_wire_05"))) {template="DISPATCH_RigVMDispatch_Constant(io Value)"}
   n2 = &RigUnit_GetTransformItemArray(Space=GlobalSpace, bInitial=False, Items=@ItemArray.Value)
   RigVMVariableNode(
    Value=&RigUnit_ControlRigSplineFromTransforms(Compression=0, SamplesPerSegment=2, SplineMode=Hermite, Stretch=0, bClosed=False, Transforms=@n2.Transforms).Spline
    ExecuteContext=@n1.ExecutePin
   ) {variable=WireSpline}
   > Create Bone Chain = RigVMFunctionReferenceNode(
    Debug=false
    Debug Scale=15
    Debug Thickness=0.5
    Enable Orient Bones=true
    Kind=Direction
    Name=CR_Spline_Headphone_Wire
    Parent=(Type=Bone,Name="dyn_headphones")
    Primary Axis=(X=1,Y=0,Z=0)
    Secondary Axis=(X=0,Y=1,Z=0)
    Transforms=@n2.Transforms
   )
   > /Script/ControlRig.RigUnit_SetTransformItemArray(Space=GlobalSpace, Weight=1, bInitial=true, bPropagateToChildren=false, Items=@"Create Bone Chain".Bones, Transforms=@n2.Transforms) {is=struct RigVMUnitNode}
   > RigVMVariableNode(Value=@"Create Bone Chain".Bones) {variable=SplineBones}
  ]}
]
hierarchy [
 root {loc=<0 0 0>, rot=<0 0 0 1>}
 pelvis {parent=root, loc=<0 -0.5301 98.0571>, rot=<0.0127 0.707 -0.0127 -0.707>}
 spine_01 {parent=pelvis, loc=<6.1208 -0.9328 0>, rot=<0 0 0.034 -0.9994>}
 spine_02 {parent=spine_01, loc=<6.5997 -0.2095 0>, rot=<0 0 0.0238 -0.9997>}
 spine_03 {parent=spine_02, loc=<6.4914 -0.1237 0>, rot=<0 0 0.0149 -0.9999>}
 spine_04 {parent=spine_03, loc=<6.9598 0.0794 0>, rot=<0 0 -0.0132 -0.9999>}
 spine_05 {parent=spine_04, loc=<6.9987 0 0>, rot=<0 0 -0.0015 -1>}
 neck_01 {parent=spine_05, loc=<11.3138 0.1529 0>, rot=<0 0 -0.2003 -0.9797>}
 neck_02 {parent=neck_01, loc=<4.5627 0 0>, rot=<0 0 0.0504 -0.9987>}
 head {parent=neck_02, loc=<5.3914 0 0>, rot=<0 0 0.115 -0.9934>}
 clavicle_l {parent=spine_05, loc=<5.4831 -0.1053 -3.2946>, rot=<0.0465 -0.7289 -0.009 -0.6829>}
 upperarm_l {parent=clavicle_l, loc=<9.8658 -1.2136 -0.4105>, rot=<0.0164 -0.3245 -0.0269 -0.9454>}
 lowerarm_l {parent=upperarm_l, loc=<26.465 -0.8345 0.259>, rot=<0.0036 0.0161 -0.2298 -0.9731>}
 hand_l {parent=lowerarm_l, loc=<22.7355 -0.3203 -0.6105>, rot=<-0.6784 -0.0223 0.0735 -0.7307>}
 thumb_01_l {parent=hand_l, loc=<1.3118 -0.3191 -2.459>, rot=<0.6546 -0.4042 0.0664 -0.6354>}
 thumb_02_l {parent=thumb_01_l, loc=<5.9336 -0.0448 0.0401>, rot=<-0.0037 0.0213 0.298 -0.9543>}
 thumb_03_l {parent=thumb_02_l, loc=<2.2223 0.3682 -0.0907>, rot=<0.0118 -0.013 0.0066 -0.9998>}
 index_metacarpal_l {parent=hand_l, loc=<1.6535 0.0183 -2.96>, rot=<0.0158 -0.0124 -0.0173 -0.9996>}
 index_01_l {parent=index_metacarpal_l, loc=<8.3008 0.0357 -0.0524>, rot=<-0.0044 0.0485 0.0924 -0.9945>}
 index_02_l {parent=index_01_l, loc=<3.3587 -0.3004 0.0452>, rot=<-0.002 0.0027 0.3225 -0.9466>}
 index_03_l {parent=index_02_l, loc=<2.5578 -0.123 -0.0252>, rot=<-0.002 0.0067 0.2558 -0.9667>}
 middle_metacarpal_l {parent=hand_l, loc=<2.1856 -0.1725 -0.5984>, rot=<0.006 0.0168 -0.0329 -0.9993>}
 middle_01_l {parent=middle_metacarpal_l, loc=<7.3275 -0.2447 0.0129>, rot=<-0.0208 0.0042 0.0858 -0.9961>}
 middle_02_l {parent=middle_01_l, loc=<4.0123 0.1401 0.1175>, rot=<-0.0218 0.0258 0.3512 -0.9357>}
 middle_03_l {parent=middle_02_l, loc=<2.6077 0.0534 0.1373>, rot=<0.0082 0.0188 0.1708 -0.9851>}
 ring_metacarpal_l {parent=hand_l, loc=<2.2074 -0.3625 1.3638>, rot=<-0.0284 0.0722 -0.0458 -0.9959>}
 ring_01_l {parent=ring_metacarpal_l, loc=<7.0295 -0.6197 -0.1446>, rot=<-0.0489 -0.0252 0.1892 -0.9804>}
 ring_02_l {parent=ring_01_l, loc=<3.6174 0.4647 -0.1027>, rot=<0 0 0.2806 -0.9598>}
 ring_03_l {parent=ring_02_l, loc=<2.4152 0.114 0.0199>, rot=<0 0 0.1189 -0.9929>}
 pinky_metacarpal_l {parent=hand_l, loc=<1.7616 -0.9753 3.3814>, rot=<-0.0397 0.0814 -0.069 -0.9935>}
 pinky_01_l {parent=pinky_metacarpal_l, loc=<7.1709 -0.4831 -0.1094>, rot=<-0.0654 -0.0035 0.2242 -0.9723>}
 pinky_02_l {parent=pinky_01_l, loc=<2.8274 0.0893 -0.1043>, rot=<0.0037 0.0011 0.2854 -0.9584>}
 pinky_03_l {parent=pinky_02_l, loc=<2.2046 0.0307 -0.0078>, rot=<-0.0002 0.0098 0.1445 -0.9895>}
 deform_wrist_dn_base_l {parent=hand_l, loc=<0 -1.5326 0>, rot=<0 0 0 -1>}
 deform_wrist_dn_l {parent=deform_wrist_dn_base_l, loc=<0 0 0>, rot=<0 0 0 -1>}
 deform_wrist_up_base_l {parent=hand_l, loc=<0 2.3009 0>, rot=<0 0 0 -1>}
 deform_wrist_up_l {parent=deform_wrist_up_base_l, loc=<0 0 0>, rot=<0 0 0 -1>}
 lowerarm_twist_02_l {parent=lowerarm_l, loc=<10.6707 0 0>, rot=<0 0 0 1>}
 lowerarm_twist_01_l {parent=lowerarm_l, loc=<18.316 0 0>, rot=<0 0 0 1>}
 deform_elbow_in_base_l {parent=lowerarm_l, loc=<0.1834 3.6062 -0.2572>, rot=<0 0 0 -1>}
 deform_elbow_in_l {parent=deform_elbow_in_base_l, loc=<0 0 0>, rot=<0 0 0 -1>}
 deform_elbow_out_base_l {parent=lowerarm_l, loc=<0 -4 0>, rot=<0 0 0 1>}
 deform_elbow_out_l {parent=deform_elbow_out_base_l, loc=<0 0 0>, rot=<0 0 0 1>}
 upperarm_twist_01_l {parent=upperarm_l, loc=<7.2135 -0.1134 0>, rot=<0 0 0 1>}
 deform_upperarm_bk_base_l {parent=upperarm_twist_01_l, loc=<-6.4053 -4.3746 0.314>, rot=<0 0 0 -1>}
 deform_upperarm_bk_l {parent=deform_upperarm_bk_base_l, loc=<0 0 0>, rot=<0 0 0 -1>}
 deform_upperarm_fr_base_l {parent=upperarm_twist_01_l, loc=<-5.6312 5.092 0.3354>, rot=<0 0 0 -1>}
 deform_upperarm_fr_l {parent=deform_upperarm_fr_base_l, loc=<0 0 0>, rot=<0 0 0 -1>}
 deform_upperarm_up_base_l {parent=upperarm_twist_01_l, loc=<-8.891 0.4825 3.9517>, rot=<0 0 0 -1>}
 deform_upperarm_up_l {parent=deform_upperarm_up_base_l, loc=<0 0 0>, rot=<0 0 0 -1>}
 upperarm_twist_02_l {parent=upperarm_l, loc=<19.7998 -0.1134 0>, rot=<0 0 0 1>}
 clavicle_r {parent=spine_05, loc=<5.4831 -0.1053 3.2946>, rot=<0.7289 0.0465 -0.6829 0.009>}
 upperarm_r {parent=clavicle_r, loc=<-9.8658 1.2136 0.4105>, rot=<0.0164 -0.3245 -0.0269 -0.9454>}
 lowerarm_r {parent=upperarm_r, loc=<-26.465 0.8345 -0.259>, rot=<0.0036 0.0161 -0.2298 -0.9731>}
 hand_r {parent=lowerarm_r, loc=<-22.7355 0.3203 0.6106>, rot=<-0.6784 -0.0223 0.0735 -0.7307>}
 thumb_01_r {parent=hand_r, loc=<-1.3118 0.3191 2.459>, rot=<0.6546 -0.4042 0.0664 -0.6354>}
 thumb_02_r {parent=thumb_01_r, loc=<-5.9336 0.0448 -0.0401>, rot=<-0.0037 0.0213 0.298 -0.9543>}
 thumb_03_r {parent=thumb_02_r, loc=<-2.2223 -0.3682 0.0907>, rot=<0.0118 -0.013 0.0066 -0.9998>}
 index_metacarpal_r {parent=hand_r, loc=<-1.6535 -0.0183 2.96>, rot=<0.0158 -0.0124 -0.0173 -0.9996>}
 index_01_r {parent=index_metacarpal_r, loc=<-8.3008 -0.0357 0.0524>, rot=<-0.0044 0.0485 0.0924 -0.9945>}
 index_02_r {parent=index_01_r, loc=<-3.3587 0.3004 -0.0452>, rot=<-0.002 0.0027 0.3225 -0.9466>}
 index_03_r {parent=index_02_r, loc=<-2.5578 0.123 0.0252>, rot=<-0.002 0.0067 0.2558 -0.9667>}
 middle_metacarpal_r {parent=hand_r, loc=<-2.1856 0.1725 0.5984>, rot=<0.006 0.0168 -0.0329 -0.9993>}
 middle_01_r {parent=middle_metacarpal_r, loc=<-7.3275 0.2447 -0.0129>, rot=<-0.0208 0.0042 0.0858 -0.9961>}
 middle_02_r {parent=middle_01_r, loc=<-4.0123 -0.1401 -0.1175>, rot=<-0.0218 0.0258 0.3512 -0.9357>}
 middle_03_r {parent=middle_02_r, loc=<-2.6077 -0.0534 -0.1373>, rot=<0.0082 0.0188 0.1708 -0.9851>}
 ring_metacarpal_r {parent=hand_r, loc=<-2.2074 0.3625 -1.3638>, rot=<-0.0284 0.0722 -0.0458 -0.9959>}
 ring_01_r {parent=ring_metacarpal_r, loc=<-7.0295 0.6197 0.1446>, rot=<-0.0489 -0.0252 0.1892 -0.9804>}
 ring_02_r {parent=ring_01_r, loc=<-3.6174 -0.4647 0.1027>, rot=<0 0 0.2806 -0.9598>}
 ring_03_r {parent=ring_02_r, loc=<-2.4152 -0.114 -0.0199>, rot=<0 0 0.1189 -0.9929>}
 pinky_metacarpal_r {parent=hand_r, loc=<-1.7616 0.9753 -3.3814>, rot=<-0.0397 0.0814 -0.069 -0.9935>}
 pinky_01_r {parent=pinky_metacarpal_r, loc=<-7.1709 0.4831 0.1094>, rot=<-0.0654 -0.0035 0.2242 -0.9723>}
 pinky_02_r {parent=pinky_01_r, loc=<-2.8274 -0.0893 0.1043>, rot=<0.0037 0.0011 0.2854 -0.9584>}
 pinky_03_r {parent=pinky_02_r, loc=<-2.2046 -0.0307 0.0078>, rot=<-0.0002 0.0098 0.1445 -0.9895>}
 deform_wrist_dn_base_r {parent=hand_r, loc=<-0.0001 1.5328 0>, rot=<0 0 0 -1>}
 deform_wrist_dn_r {parent=deform_wrist_dn_base_r, loc=<0 0 0>, rot=<0 0 0 -1>}
 deform_wrist_up_base_r {parent=hand_r, loc=<-0.0002 -2.3007 -0.0001>, rot=<0 0 0 -1>}
 deform_wrist_up_r {parent=deform_wrist_up_base_r, loc=<0 0 0>, rot=<0 0 0 -1>}
 lowerarm_twist_02_r {parent=lowerarm_r, loc=<-10.6707 0 0>, rot=<0 0 0 -1>}
 lowerarm_twist_01_r {parent=lowerarm_r, loc=<-18.316 0 0>, rot=<0 0 0 -1>}
 deform_elbow_in_base_r {parent=lowerarm_r, loc=<-0.1834 -3.6062 0.2572>, rot=<0 0 0 -1>}
 deform_elbow_in_r {parent=deform_elbow_in_base_r, loc=<0 0 0>, rot=<0 0 0 -1>}
 deform_elbow_out_base_r {parent=lowerarm_r, loc=<0 4 -0.0001>, rot=<0 0 0 -1>}
 deform_elbow_out_r {parent=deform_elbow_out_base_r, loc=<0 0 0>, rot=<0 0 0 1>}
 upperarm_twist_01_r {parent=upperarm_r, loc=<-7.2135 0.1134 0>, rot=<0 0 0 -1>}
 deform_upperarm_bk_base_r {parent=upperarm_twist_01_r, loc=<6.4053 4.3746 -0.314>, rot=<0 0 0 -1>}
 deform_upperarm_bk_r {parent=deform_upperarm_bk_base_r, loc=<0 0 0>, rot=<0 0 0 -1>}
 deform_upperarm_fr_base_r {parent=upperarm_twist_01_r, loc=<5.6312 -5.092 -0.3354>, rot=<0 0 0 -1>}
 deform_upperarm_fr_r {parent=deform_upperarm_fr_base_r, loc=<0 0 0>, rot=<0 0 0 -1>}
 deform_upperarm_up_base_r {parent=upperarm_twist_01_r, loc=<8.891 -0.4825 -3.9517>, rot=<0 0 0 -1>}
 deform_upperarm_up_r {parent=deform_upperarm_up_base_r, loc=<0 0 0>, rot=<0 0 0 -1>}
 upperarm_twist_02_r {parent=upperarm_r, loc=<-19.7998 0.1134 0>, rot=<0 0 0 -1>}
 attach_backpack {parent=spine_05, loc=<-0.123 -13.473 0>, rot=<-0.0093 -0.707 0.0093 -0.707>}
 Backpack_BR {parent=attach_backpack, type=Null, loc=<0 0.1195 0>, rot=<-0.0581 0 0 0.9983>}
 attach_cape {parent=spine_05, loc=<9.017 -11.714 0>, rot=<0.0616 -0.7044 -0.0616 -0.7044>}
 cape {parent=attach_cape, type=Null, loc=<3.3056 1.2833 -1.6301>, rot=<-0.0433 -0.0073 -0.0019 0.999>}
 dyn_hood_base {parent=spine_05, loc=<4.9662 -9.6498 0>, rot=<0.1551 -0.9879 0 0>}
 dyn_hood_01 {parent=dyn_hood_base, loc=<0.2 0 0>, rot=<0 0 0 1>}
 dyn_hood_02 {parent=dyn_hood_01, loc=<6.4623 0 0>, rot=<0 0 -0.0607 -0.9982>}
 dyn_hood_03 {parent=dyn_hood_02, loc=<4.6523 0 0>, rot=<0 0 0 1>}
 dyn_headphones_base {parent=spine_05, loc=<13.5317 -3.4048 0.0021>, rot=<-0.0035 0.0111 -0.8299 -0.5578>}
 dyn_headphones {parent=dyn_headphones_base, loc=<-0.2 0 0>, rot=<0 0 0 1>}
 deform_headphone_wire_01 {parent=dyn_headphones, loc=<19.9772 7.064 0.2092>, rot=<0.0434 -0.1019 -0.4152 -0.903>}
 deform_headphone_wire_02 {parent=deform_headphone_wire_01, loc=<4.9039 0 0>, rot=<0.1782 -0.0656 -0.2712 -0.9436>}
 deform_headphone_wire_03 {parent=deform_headphone_wire_02, loc=<4.651 0 0>, rot=<-0.0083 -0.0481 -0.0789 -0.9957>}
 deform_headphone_wire_04 {parent=deform_headphone_wire_03, loc=<4.5342 0 0>, rot=<-0.1579 -0.0149 -0.0402 -0.9865>}
 deform_headphone_wire_05 {parent=deform_headphone_wire_04, loc=<4.447 0 0>, rot=<-0.1926 0.024 -0.0505 -0.9797>}
 dyn_headphone_wire_04_c {parent=deform_headphone_wire_04, loc=<0.5658 0 0>, rot=<0 0 0 1>}
 dyn_headphone_wire_03_c {parent=deform_headphone_wire_03, loc=<0.4865 -0.03 0.0492>, rot=<0 0 0 1>}
 dyn_headphone_wire_02_l {parent=deform_headphone_wire_02, loc=<0.7623 0.1841 -1.374>, rot=<-0.0075 0.0943 0.0788 -0.9924>}
 dyn_headphone_wire_02_r {parent=deform_headphone_wire_02, loc=<0.9975 -0.5173 1.3964>, rot=<0 -0.044 0 -0.999>}
 dyn_headphone_wire_01_r {parent=deform_headphone_wire_01, loc=<0.6225 -0.579 1.9193>, rot=<0 0 0 1>}
 dyn_headphone_wire_01_l {parent=deform_headphone_wire_01, loc=<1.9176 -0.6996 -1.6936>, rot=<0.0153 0.0432 -0.0163 -0.9988>}
 deform_pec_aim_base_l {parent=spine_04, loc=<3.2456 9.3309 -1.2237>, rot=<0.0603 -0.4033 0.3318 -0.8507>}
 deform_pec_aim_l {parent=deform_pec_aim_base_l, loc=<0 0 0>, rot=<0 0 0 1>}
 deform_pec_base_l {parent=deform_pec_aim_l, loc=<9.3665 0 0>, rot=<0 0 0 1>}
 deform_pec_l {parent=deform_pec_base_l, loc=<0 0 0>, rot=<0 0 0 -1>}
 deform_pec_aim_base_r {parent=spine_04, loc=<3.2452 9.3309 1.2237>, rot=<-0.4033 -0.0603 0.8507 0.3318>}
 deform_pec_aim_r {parent=deform_pec_aim_base_r, loc=<0 0 0>, rot=<0 0 0 1>}
 deform_pec_base_r {parent=deform_pec_aim_r, loc=<-9.3667 -0.0003 -0.0004>, rot=<0 0 0 1>}
 deform_pec_r {parent=deform_pec_base_r, loc=<0 0 0>, rot=<0 0 0 -1>}
 dyn_lapel_l_base {parent=spine_04, loc=<5.3418 11.8078 -8.6251>, rot=<-0.6065 0.3354 0.7093 -0.1285>}
 dyn_lapel_l_01 {parent=dyn_lapel_l_base, loc=<0.2 0 0>, rot=<0 0 0 1>}
 dyn_lapel_l_02 {parent=dyn_lapel_l_01, loc=<3.9666 0 0>, rot=<0 0 0 1>}
 dyn_lapel_r_base {parent=spine_04, loc=<6.2931 11.008 8.7982>, rot=<0.6856 0.0967 -0.6599 -0.2916>}
 dyn_lapel_r_01 {parent=dyn_lapel_r_base, loc=<-0.2 0 0>, rot=<0 0 0 1>}
 dyn_lapel_r_02 {parent=dyn_lapel_r_01, loc=<-3.9665 -0.0002 -0.0005>, rot=<0 0 0 1>}
 deform_armpit_l_base {parent=spine_04, loc=<-1.3702 -2.8656 -15.3873>, rot=<0.0289 -0.713 -0.0367 -0.6996>}
 deform_armpit_l {parent=deform_armpit_l_base, loc=<0.2 0 0>, rot=<0 0 0 1>}
 deform_armpit_r_base {parent=spine_04, loc=<-1.3701 -2.8656 15.3873>, rot=<0.713 0.0289 -0.6996 0.0367>}
 deform_armpit_r {parent=deform_armpit_r_base, loc=<-0.2 0 0>, rot=<0 0 0 1>}
 dyn_jacket_front_l_base {parent=spine_04, loc=<-2.0884 8.8704 -11.2553>, rot=<-0.0578 0.2728 0.9603 -0.011>}
 dyn_jacket_front_l_01 {parent=dyn_jacket_front_l_base, loc=<0.2 0 0>, rot=<0 0 0 1>}
 dyn_jacket_front_l_02 {parent=dyn_jacket_front_l_01, loc=<10 0 0>, rot=<0 0 0 1>}
 dyn_jacket_front_l_03 {parent=dyn_jacket_front_l_02, loc=<10 0 0>, rot=<0 0 0 1>}
 dyn_jacket_front_l_04 {parent=dyn_jacket_front_l_03, loc=<15.5 0 0>, rot=<0 0 0 1>}
 dyn_jacket_front_l_05 {parent=dyn_jacket_front_l_04, loc=<15.5 0 0>, rot=<0 0 0 1>}
 dyn_jacket_front_l_06 {parent=dyn_jacket_front_l_05, loc=<15.5 0 0>, rot=<0 0 0 1>}
 dyn_jacket_front_l_07 {parent=dyn_jacket_front_l_06, loc=<15.5 0 0>, rot=<0 0 0 1>}
 dyn_jacket_front_r_base {parent=spine_04, loc=<-2.088 8.8704 11.2553>, rot=<-0.2728 -0.0578 -0.011 -0.9603>}
 dyn_jacket_front_r_01 {parent=dyn_jacket_front_r_base, loc=<-0.2002 0 0>, rot=<0 0 0 1>}
 dyn_jacket_front_r_02 {parent=dyn_jacket_front_r_01, loc=<-10.0007 -0.0001 0>, rot=<0 0 0 1>}
 dyn_jacket_front_r_03 {parent=dyn_jacket_front_r_02, loc=<-9.9997 0 0>, rot=<0 0 0 1>}
 dyn_jacket_front_r_04 {parent=dyn_jacket_front_r_03, loc=<-15.4998 0 0.0001>, rot=<0 0 0 1>}
 dyn_jacket_front_r_05 {parent=dyn_jacket_front_r_04, loc=<-15.5001 0 -0.0001>, rot=<0 0 0 1>}
 dyn_jacket_front_r_06 {parent=dyn_jacket_front_r_05, loc=<-15.5 0 0>, rot=<0 0 0 1>}
 dyn_jacket_front_r_07 {parent=dyn_jacket_front_r_06, loc=<-15.5 0 -0.0001>, rot=<0 0 0 1>}
 dyn_jacket_rear_l_base {parent=spine_03, loc=<-3.8282 -6.6993 -12.3503>, rot=<0.0808 -0.9287 -0.3618 -0.001>}
 dyn_jacket_rear_l_01 {parent=dyn_jacket_rear_l_base, loc=<0.2 0 0>, rot=<0 0 0 1>}
 dyn_jacket_rear_l_02 {parent=dyn_jacket_rear_l_01, loc=<10 0 0>, rot=<0 0 0.0185 -0.9998>}
 dyn_jacket_rear_l_03 {parent=dyn_jacket_rear_l_02, loc=<15.5 0 0>, rot=<0 0 0 1>}
 dyn_jacket_rear_l_04 {parent=dyn_jacket_rear_l_03, loc=<15.5 0 0>, rot=<0 0 0 1>}
 dyn_jacket_rear_l_05 {parent=dyn_jacket_rear_l_04, loc=<15.5 0 0>, rot=<0 0 0 1>}
 dyn_jacket_rear_l_06 {parent=dyn_jacket_rear_l_05, loc=<15.5 0 0>, rot=<0 0 0 1>}
 dyn_jacket_rear_base {parent=spine_03, loc=<-3.3943 -10.6478 -0.0825>, rot=<0.0926 -0.9952 0.0299 -0.008>}
 dyn_jacket_rear_01 {parent=dyn_jacket_rear_base, loc=<0.2 0 0>, rot=<0 0 0 1>}
 dyn_jacket_rear_02 {parent=dyn_jacket_rear_01, loc=<10 0 0>, rot=<0 0 0 1>}
 dyn_jacket_rear_03 {parent=dyn_jacket_rear_02, loc=<15.5 0 0>, rot=<0 0 0 1>}
 dyn_jacket_rear_04 {parent=dyn_jacket_rear_03, loc=<15.5 0 0>, rot=<0 0 0 1>}
 dyn_jacket_rear_05 {parent=dyn_jacket_rear_04, loc=<15.5 0 0>, rot=<0 0 0 1>}
 dyn_jacket_rear_06 {parent=dyn_jacket_rear_05, loc=<15.5 0 0>, rot=<0 0 0 1>}
 dyn_jacket_rear_r_base {parent=spine_03, loc=<-3.8284 -6.6994 12.3503>, rot=<0.9287 0.0808 -0.001 0.3618>}
 dyn_jacket_rear_r_01 {parent=dyn_jacket_rear_r_base, loc=<-0.1995 0 0>, rot=<0 0 0 1>}
 dyn_jacket_rear_r_02 {parent=dyn_jacket_rear_r_01, loc=<-10.0005 0 0>, rot=<0 0 0.0185 -0.9998>}
 dyn_jacket_rear_r_03 {parent=dyn_jacket_rear_r_02, loc=<-15.4996 0 0>, rot=<0 0 0 1>}
 dyn_jacket_rear_r_04 {parent=dyn_jacket_rear_r_03, loc=<-15.5 0 0>, rot=<0 0 0 1>}
 dyn_jacket_rear_r_05 {parent=dyn_jacket_rear_r_04, loc=<-15.5 0 0>, rot=<0 0 0 1>}
 dyn_jacket_rear_r_06 {parent=dyn_jacket_rear_r_05, loc=<-15.4999 0 0>, rot=<0 0 0 1>}
 def_belt {parent=spine_01, loc=<3.8216 1.141 0>, rot=<0 0 0 1>}
 thigh_l {parent=pelvis, loc=<-1.0855 -0.2109 -8.7053>, rot=<0.0009 0.0355 0.0155 -0.9993>}
 calf_l {parent=thigh_l, loc=<-42.5284 -0.0019 -0.7297>, rot=<0.0006 -0.0161 -0.0391 -0.9991>}
 foot_l {parent=calf_l, loc=<-46.7401 -2.5014 -0.592>, rot=<0.0427 -0.0238 0.0353 -0.9982>}
 ball_l {parent=foot_l, loc=<-3.5658 10.8066 -0.1544>, rot=<0.0001 -0.0001 -0.7063 -0.7079>}
 calf_twist_02_l {parent=calf_l, loc=<-19.1546 -1.7943 -0.2638>, rot=<0.0237 0.0095 -0.0035 -0.9997>}
 calf_twist_01_l {parent=calf_l, loc=<-35.8483 -1.9197 -0.5794>, rot=<0.0237 0.0095 -0.0035 -0.9997>}
 deform_knee_in_base_l {parent=calf_l, loc=<0.8031 -5.0231 0.2456>, rot=<0 0 0 -1>}
 deform_knee_in_l {parent=deform_knee_in_base_l, loc=<0 0 0>, rot=<0 0 0 -1>}
 deform_knee_out_base_l {parent=calf_l, loc=<0 4 0>, rot=<0 0 0 -1>}
 deform_knee_out_l {parent=deform_knee_out_base_l, loc=<0 0 0>, rot=<0 0 0 1>}
 thigh_twist_01_l {parent=thigh_l, loc=<-27.952 0 0>, rot=<0.0013 0 0 -1>}
 deform_groin_base_l {parent=thigh_l, loc=<0.4801 5.6776 -0.2486>, rot=<0 0 0 -1>}
 deform_groin_l {parent=deform_groin_base_l, loc=<0 0 0>, rot=<0 0 0 -1>}
 cr_thigh_l_no_twist {parent=thigh_l, loc=<0 0 0>, rot=<0 0 0 1>}
 thigh_r {parent=pelvis, loc=<-1.0855 -0.2109 8.7053>, rot=<0.0355 -0.0009 0.9993 0.0155>}
 calf_r {parent=thigh_r, loc=<42.5284 0.0019 0.7297>, rot=<0.0006 -0.0161 -0.0391 -0.9991>}
 foot_r {parent=calf_r, loc=<46.7401 2.5014 0.592>, rot=<0.0427 -0.0238 0.0353 -0.9982>}
 ball_r {parent=foot_r, loc=<3.5658 -10.8066 0.1544>, rot=<0.0001 -0.0001 -0.7063 -0.7079>}
 calf_twist_02_r {parent=calf_r, loc=<19.1546 1.7943 0.2638>, rot=<0.0237 0.0095 -0.0035 -0.9997>}
 calf_twist_01_r {parent=calf_r, loc=<35.8483 1.9197 0.5794>, rot=<0.0237 0.0095 -0.0035 -0.9997>}
 deform_knee_in_base_r {parent=calf_r, loc=<-0.8031 5.0231 -0.2457>, rot=<0 0 0 -1>}
 deform_knee_in_r {parent=deform_knee_in_base_r, loc=<0 0 0>, rot=<0 0 0 -1>}
 deform_knee_out_base_r {parent=calf_r, loc=<0 -4 0>, rot=<0 0 0 -1>}
 deform_knee_out_r {parent=deform_knee_out_base_r, loc=<0 0 0>, rot=<0 0 0 1>}
 thigh_twist_01_r {parent=thigh_r, loc=<27.952 0 0>, rot=<0.0013 0 0 -1>}
 deform_groin_base_r {parent=thigh_r, loc=<-0.4801 -5.6776 0.2486>, rot=<0 0 0 -1>}
 deform_groin_r {parent=deform_groin_base_r, loc=<0 0 0>, rot=<0 0 0 -1>}
 cr_thigh_r_no_twist {parent=thigh_r, loc=<0 0 0>, rot=<0 0 0 1>}
 deform_glute_base_l {parent=pelvis, loc=<-7.3466 -9.0213 -9.1245>, rot=<0.0172 0.031 -0.4477 -0.8935>}
 deform_glute_l {parent=deform_glute_base_l, loc=<0 0 0>, rot=<0 0 0 1>}
 deform_glute_base_r {parent=pelvis, loc=<-7.3464 -9.0213 9.1245>, rot=<0.031 -0.0172 0.8935 -0.4477>}
 deform_glute_r {parent=deform_glute_base_r, loc=<0 0 0>, rot=<0 0 0 1>}
 dyn_skirt_front_base {parent=pelvis, loc=<0.2233 9.1985 0>, rot=<0 0.0004 0.9977 0.068>}
 dyn_skirt_front_01 {parent=dyn_skirt_front_base, loc=<0.2 0 0>, rot=<0 0 0 1>}
 dyn_skirt_front_02 {parent=dyn_skirt_front_01, loc=<9.7961 0 0>, rot=<0 0 0 1>}
 dyn_skirt_front_l_base {parent=pelvis, loc=<1.0407 6.9447 -10.5791>, rot=<-0.049 0.1468 0.9828 0.1003>}
 dyn_skirt_front_l_01 {parent=dyn_skirt_front_l_base, loc=<0.2 0 0>, rot=<0 0 0 1>}
 dyn_skirt_front_l_02 {parent=dyn_skirt_front_l_01, loc=<10.7711 0.0683 0>, rot=<0 0 -0.0245 -0.9997>}
 dyn_skirt_side_l_base {parent=pelvis, loc=<2.0065 -1.3757 -13.5246>, rot=<0.0813 -0.7196 -0.6837 -0.0899>}
 dyn_skirt_side_l_01 {parent=dyn_skirt_side_l_base, loc=<0.2 0 0>, rot=<0 0 0 1>}
 dyn_skirt_side_l_02 {parent=dyn_skirt_side_l_01, loc=<12.2026 0.0443 0>, rot=<0 0 -0.0268 -0.9996>}
 dyn_skirt_rear_l_base {parent=pelvis, loc=<1.0574 -9.9494 -8.5294>, rot=<0.0442 -0.9787 -0.1965 -0.0399>}
 dyn_skirt_rear_l_01 {parent=dyn_skirt_rear_l_base, loc=<0.2 0 0>, rot=<0 0 0 1>}
 dyn_skirt_rear_l_02 {parent=dyn_skirt_rear_l_01, loc=<11.3879 0.0556 0>, rot=<0 0 -0.0298 -0.9996>}
 dyn_skirt_rear_base {parent=pelvis, loc=<0.9162 -10.5923 -0.0004>, rot=<0.0331 -0.9995 0.0004 0>}
 dyn_skirt_rear_01 {parent=dyn_skirt_rear_base, loc=<0.2 0 0>, rot=<0 0 0 1>}
 dyn_skirt_rear_02 {parent=dyn_skirt_rear_01, loc=<11.1342 0 0>, rot=<0 0 -0.0428 -0.9991>}
 dyn_skirt_front_r_base {parent=pelvis, loc=<1.0407 6.9447 10.5791>, rot=<-0.1468 -0.049 0.1003 -0.9828>}
 dyn_skirt_front_r_01 {parent=dyn_skirt_front_r_base, loc=<-0.2 0 0>, rot=<0 0 0 -1>}
 dyn_skirt_front_r_02 {parent=dyn_skirt_front_r_01, loc=<-10.7711 -0.0683 0>, rot=<0 0 -0.0245 -0.9997>}
 dyn_skirt_side_r_base {parent=pelvis, loc=<2.0069 -1.3757 13.5246>, rot=<-0.7196 -0.0813 0.0899 -0.6837>}
 dyn_skirt_side_r_01 {parent=dyn_skirt_side_r_base, loc=<-0.2005 0 0>, rot=<0 0 0 1>}
 dyn_skirt_side_r_02 {parent=dyn_skirt_side_r_01, loc=<-12.2025 -0.0443 0>, rot=<0 0 -0.0268 -0.9996>}
 dyn_skirt_rear_r_base {parent=pelvis, loc=<1.0574 -9.9494 8.5294>, rot=<0.9787 0.0442 -0.0399 0.1965>}
 dyn_skirt_rear_r_01 {parent=dyn_skirt_rear_r_base, loc=<-0.2 0 0>, rot=<0 0 0 -1>}
 dyn_skirt_rear_r_02 {parent=dyn_skirt_rear_r_01, loc=<-11.3879 -0.0557 0>, rot=<0 0 -0.0298 -0.9996>}
 def_slope_warp {parent=root, loc=<0 0 0>, rot=<0 0 0 1>}
 def_simspace {parent=root, loc=<0 0 0>, rot=<0 0 0 1>}
]
variables [
 ClothEnabled {type=bool, default=False}
 InSplineTransforms {type="TArray<FTransform>"}
 IsBaseballSliding {type=bool, default=False}
 IsKneeSliding {type=bool, default=False}
 SPR_Thigh_L_Back {type=double, default=0}
 SPR_Thigh_L_Inbetween_Up {type=double, default=0}
 SPR_Thigh_L_Side {type=double, default=0}
 SPR_Thigh_L_Up {type=double, default=0}
 SPR_Thigh_R_Back {type=double, default=0}
 SPR_Thigh_R_Inbetween_Up {type=double, default=0}
 SPR_Thigh_R_Side {type=double, default=0}
 SPR_Thigh_R_Up {type=double, default=0}
 SplineBones {type="TArray<FRigElementKey>"}
 WireSpline {type=FControlRigSpline, default=()}
]
