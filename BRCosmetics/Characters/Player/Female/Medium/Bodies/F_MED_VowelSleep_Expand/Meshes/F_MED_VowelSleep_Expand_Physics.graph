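# Physics asset for the F_MED_VowelSleep_Expand female body cosmetic, setting up simulation of its long jacket.
# Five capsule chains (front left/right, rear center/left/right) hang from tiny base bodies with tight swing and twist limits, looser on the lower front-left.
# Free-angular cross constraints tie neighboring chains together mid-length, linking the panels.
# Pelvis, thigh and calf capsules provide body shapes, and a large sphere below def_slope_warp likely approximates the ground.
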PhysicsAsset /BRCosmetics/Characters/Player/Female/Medium/Bodies/F_MED_VowelSleep_Expand/Meshes/F_MED_VowelSleep_Expand_Physics
bodies {
 calf_l [{shapeType=Capsule, length=43.937, radius=8.6772, center=<-31.3535 0 0.9492>, rotation=<rot 0 90 -90>}]
 calf_r [{shapeType=Capsule, length=43.937, radius=8.6772, center=<31.3535 0 -0.9492>, rotation=<rot 0 90 90>}]
 def_slope_warp [{shapeType=Sphere, radius=500, shapeName=def_slope_warp_sphere, center=<0 0 -490>}]
 dyn_jacket_front_l_01 [{shapeType=Capsule, length=10, radius=8, shapeName=dyn_jacket_front_l_01_capsule, center=<8.4299 -7.5251 1.3786>, rotation=<rot 0 0 -11.8244>}]
 dyn_jacket_front_l_02 [{
   shapeType=Capsule
   length=9.975
   radius=8.025
   shapeName=dyn_jacket_front_l_02_capsule
   center=<8.8197 -6.986 1.5118>
   rotation=<rot 0.0001 -0.0001 -7.3699>
  }]
 dyn_jacket_front_l_03 [{
   shapeType=Capsule
   length=10
   radius=8
   shapeName=dyn_jacket_front_l_03_capsule
   center=<9.5905 -6.9469 0.7127>
   rotation=<rot 0.0001 -0.0001 -2.5532>
  }]
 dyn_jacket_front_l_04 [{
   shapeType=Capsule
   length=25.1995
   radius=8
   shapeName=dyn_jacket_front_l_04_capsule
   center=<9.1107 -6.6147 1.9059>
   rotation=<rot 0 -0.0001 -2.127>
  }]
 dyn_jacket_front_l_05 [{shapeType=Capsule, length=23.1711, radius=10.0284, shapeName=dyn_jacket_front_l_05_capsule, center=<7.5766 -8.688 0>}]
 dyn_jacket_front_l_06 [{
   shapeType=Capsule
   length=24.3912
   radius=10.0824
   shapeName=dyn_jacket_front_l_06_capsule
   center=<11.0691 -7.2775 -0.1621>
   rotation=<rot 12.7297 -90 0>
  }]
 dyn_jacket_front_l_07 [{
   shapeType=Capsule
   length=28.7385
   radius=9.6573
   shapeName=dyn_jacket_front_l_07_capsule
   center=<10.2409 -4.515 -0.2513>
   rotation=<rot 0 -90 0>
  }]
 dyn_jacket_front_l_base [{shapeType=Capsule, length=0.505, radius=0.505, shapeName=dyn_jacket_front_l_base_capsule, rotation=<rot 0 -90 0>}]
 dyn_jacket_front_r_01 [{
   shapeType=Capsule
   length=10
   radius=8
   shapeName=dyn_jacket_front_r_01_capsule
   center=<-8.4299 7.5251 -1.3786>
   rotation=<rot 0 0 168.1756>
  }]
 dyn_jacket_front_r_02 [{
   shapeType=Capsule
   length=9.975
   radius=8.025
   shapeName=dyn_jacket_front_r_02_capsule
   center=<-8.8197 6.986 -1.5118>
   rotation=<rot 0.0001 -0.0001 172.6301>
  }]
 dyn_jacket_front_r_03 [{
   shapeType=Capsule
   length=10
   radius=8
   shapeName=dyn_jacket_front_r_03_capsule
   center=<-9.5905 6.9469 -0.7127>
   rotation=<rot 0.0001 -0.0001 177.4468>
  }]
 dyn_jacket_front_r_04 [{
   shapeType=Capsule
   length=25.1995
   radius=8
   shapeName=dyn_jacket_front_r_04_capsule
   center=<-9.1107 6.6147 -1.9059>
   rotation=<rot 0 -0.0001 177.873>
  }]
 dyn_jacket_front_r_05 [{
   shapeType=Capsule
   length=23.1802
   radius=10.0193
   shapeName=dyn_jacket_front_r_05_capsule
   center=<-7.5766 8.5999 0>
   rotation=<rot 0 0 -180>
  }]
 dyn_jacket_front_r_06 [{
   shapeType=Capsule
   length=24.3912
   radius=10.0824
   shapeName=dyn_jacket_front_r_06_capsule
   center=<-11.0691 7.2775 0.1621>
   rotation=<rot 12.7297 -90 180>
  }]
 dyn_jacket_front_r_07 [{
   shapeType=Capsule
   length=28.7385
   radius=9.6573
   shapeName=dyn_jacket_front_r_07_capsule
   center=<-10.2409 4.515 0.2513>
   rotation=<rot 0 -90 180>
  }]
 dyn_jacket_front_r_base [{shapeType=Capsule, length=0.505, radius=0.505, shapeName=dyn_jacket_front_r_base_capsule, rotation=<rot 0 -90 0>}]
 dyn_jacket_rear_01 [{shapeType=Capsule, length=10, radius=8, shapeName=dyn_jacket_rear_01_capsule, center=<7.9458 -0.5361 -0.0953>}]
 dyn_jacket_rear_02 [{shapeType=Capsule, length=10, radius=8, shapeName=dyn_jacket_rear_02_capsule, center=<14.6334 -2.3854 -0.2371>}]
 dyn_jacket_rear_03 [{shapeType=Capsule, length=37.8589, radius=11.2323, shapeName=dyn_jacket_rear_03_capsule, center=<7.9365 -6.3637 -0.8189>}]
 dyn_jacket_rear_04 [{shapeType=Capsule, length=32.7622, radius=16.3291, shapeName=dyn_jacket_rear_04_capsule, center=<6.9658 -3.6784 -0.7624>}]
 dyn_jacket_rear_05 [{shapeType=Capsule, length=36.4828, radius=12.6084, shapeName=dyn_jacket_rear_05_capsule, center=<5.6305 -6.3116 -0.9792>}]
 dyn_jacket_rear_06 [{
   shapeType=Capsule
   length=50.5872
   radius=11.8277
   shapeName=dyn_jacket_rear_06_capsule
   center=<8.0598 -5.9944 -0.5729>
   rotation=<rot 0 -90 0>
  }]
 dyn_jacket_rear_base [{shapeType=Capsule, length=0.505, radius=0.505, shapeName=dyn_jacket_rear_base_capsule, rotation=<rot 0 -90 0>}]
 dyn_jacket_rear_l_01 [{shapeType=Capsule, length=10, radius=8, shapeName=dyn_jacket_rear_l_01_capsule, center=<6.09 -1.7377 0>}]
 dyn_jacket_rear_l_02 [{shapeType=Capsule, length=10, radius=8, shapeName=dyn_jacket_rear_l_02_capsule, center=<13.2388 -6.3585 0>}]
 dyn_jacket_rear_l_03 [{shapeType=Capsule, length=19.395, radius=8.2646, shapeName=dyn_jacket_rear_l_03_capsule, center=<8.099 -6.0059 0>}]
 dyn_jacket_rear_l_04 [{
   shapeType=Capsule
   length=17.2999
   radius=10.3597
   shapeName=dyn_jacket_rear_l_04_capsule
   center=<8.0926 -9.4968 0.1914>
   rotation=<rot 0 0 3.6552>
  }]
 dyn_jacket_rear_l_05 [{shapeType=Capsule, length=17.7644, radius=9.8952, shapeName=dyn_jacket_rear_l_05_capsule, center=<6.4975 -6.4632 0>}]
 dyn_jacket_rear_l_06 [{shapeType=Capsule, length=22.4911, radius=9.9461, shapeName=dyn_jacket_rear_l_06_capsule, center=<9.1024 -6.2646 0>, rotation=<rot 0 -90 0>}]
 dyn_jacket_rear_l_base [{shapeType=Capsule, length=0.505, radius=0.505, shapeName=dyn_jacket_rear_l_base_capsule, rotation=<rot 0 -90 0>}]
 dyn_jacket_rear_r_01 [{shapeType=Capsule, length=10, radius=8, shapeName=dyn_jacket_rear_r_01_capsule, center=<-6.09 1.7377 0>, rotation=<rot 0 0 -180>}]
 dyn_jacket_rear_r_02 [{shapeType=Capsule, length=10, radius=8, shapeName=dyn_jacket_rear_r_02_capsule, center=<-13.2388 6.3585 0>, rotation=<rot 0 0 -180>}]
 dyn_jacket_rear_r_03 [{shapeType=Capsule, length=19.395, radius=8.2646, shapeName=dyn_jacket_rear_r_03_capsule, center=<-8.099 6.0059 0>, rotation=<rot 0 0 -180>}]
 dyn_jacket_rear_r_04 [{
   shapeType=Capsule
   length=17.1492
   radius=10.5104
   shapeName=dyn_jacket_rear_r_04_capsule
   center=<-8.0926 9.3899 -0.1846>
   rotation=<rot 0 0 -176.3448>
  }]
 dyn_jacket_rear_r_05 [{shapeType=Capsule, length=17.7644, radius=9.8952, shapeName=dyn_jacket_rear_r_05_capsule, center=<-6.4975 6.4632 0>, rotation=<rot 0 0 -180>}]
 dyn_jacket_rear_r_06 [{
   shapeType=Capsule
   length=22.4911
   radius=9.9461
   shapeName=dyn_jacket_rear_r_06_capsule
   center=<-9.1024 6.2646 0>
   rotation=<rot 0 -90 180>
  }]
 dyn_jacket_rear_r_base [{shapeType=Capsule, length=0.505, radius=0.505, shapeName=dyn_jacket_rear_r_base_capsule, rotation=<rot 0 -90 0>}]
 pelvis [{shapeType=Capsule, length=1.8788, radius=13.7769, center=<-3.7275 3.5183 -0.0531>, rotation=<rot 0 -90.0001 -90.8156>}]
 thigh_l [{shapeType=Capsule, length=32.9001, radius=10.6762, center=<-22.288 0 0>, rotation=<rot 0 89.9995 -88.1001>}]
 thigh_r [{shapeType=Capsule, length=32.9001, radius=10.6762, center=<22.288 0 0>, rotation=<rot 0 89.9995 91.8999>}]
}
constraints [
 {bone1Name=dyn_jacket_front_l_01, bone2Name=dyn_jacket_front_l_base, swing1LimitDegrees=12, swing2LimitDegrees=8, twistLimitDegrees=4}
 {bone1Name=dyn_jacket_front_l_02, bone2Name=dyn_jacket_front_l_01, swing1LimitDegrees=12, swing2LimitDegrees=8, twistLimitDegrees=4}
 {bone1Name=dyn_jacket_front_l_03, bone2Name=dyn_jacket_front_l_02, swing1LimitDegrees=12, swing2LimitDegrees=8, twistLimitDegrees=4}
 {bone1Name=dyn_jacket_front_l_04, bone2Name=dyn_jacket_front_l_03, swing1LimitDegrees=12, swing2LimitDegrees=8, twistLimitDegrees=4}
 {bone1Name=dyn_jacket_front_l_05, bone2Name=dyn_jacket_front_l_04, swing1LimitDegrees=24, swing2LimitDegrees=8, twistLimitDegrees=4}
 {bone1Name=dyn_jacket_front_r_01, bone2Name=dyn_jacket_front_r_base, swing1LimitDegrees=12, swing2LimitDegrees=8, twistLimitDegrees=4}
 {bone1Name=dyn_jacket_front_r_02, bone2Name=dyn_jacket_front_r_01, swing1LimitDegrees=12, swing2LimitDegrees=8, twistLimitDegrees=4}
 {bone1Name=dyn_jacket_front_r_03, bone2Name=dyn_jacket_front_r_02, swing1LimitDegrees=12, swing2LimitDegrees=8, twistLimitDegrees=4}
 {bone1Name=dyn_jacket_front_r_04, bone2Name=dyn_jacket_front_r_03, swing1LimitDegrees=12, swing2LimitDegrees=8, twistLimitDegrees=4}
 {bone1Name=dyn_jacket_front_r_05, bone2Name=dyn_jacket_front_r_04, swing1LimitDegrees=12, swing2LimitDegrees=8, twistLimitDegrees=4}
 {bone1Name=dyn_jacket_rear_01, bone2Name=dyn_jacket_rear_base, swing1LimitDegrees=12, swing2LimitDegrees=8, twistLimitDegrees=4}
 {bone1Name=dyn_jacket_rear_02, bone2Name=dyn_jacket_rear_01, swing1LimitDegrees=12, swing2LimitDegrees=8, twistLimitDegrees=4}
 {bone1Name=dyn_jacket_rear_03, bone2Name=dyn_jacket_rear_02, swing1LimitDegrees=12, swing2LimitDegrees=8, twistLimitDegrees=4}
 {bone1Name=dyn_jacket_rear_04, bone2Name=dyn_jacket_rear_03, swing1LimitDegrees=12, swing2LimitDegrees=8, twistLimitDegrees=4}
 {bone1Name=dyn_jacket_rear_05, bone2Name=dyn_jacket_rear_04, swing1LimitDegrees=12, swing2LimitDegrees=8, twistLimitDegrees=4}
 {bone1Name=dyn_jacket_rear_l_01, bone2Name=dyn_jacket_rear_l_base, swing1LimitDegrees=12, swing2LimitDegrees=8, twistLimitDegrees=4}
 {bone1Name=dyn_jacket_rear_l_02, bone2Name=dyn_jacket_rear_l_01, swing1LimitDegrees=12, swing2LimitDegrees=8, twistLimitDegrees=4}
 {bone1Name=dyn_jacket_rear_l_03, bone2Name=dyn_jacket_rear_l_02, swing1LimitDegrees=12, swing2LimitDegrees=8, twistLimitDegrees=4}
 {bone1Name=dyn_jacket_rear_l_04, bone2Name=dyn_jacket_rear_l_03, swing1LimitDegrees=12, swing2LimitDegrees=8, twistLimitDegrees=4}
 {bone1Name=dyn_jacket_rear_l_05, bone2Name=dyn_jacket_rear_l_04, swing1LimitDegrees=12, swing2LimitDegrees=8, twistLimitDegrees=4}
 {bone1Name=dyn_jacket_rear_r_01, bone2Name=dyn_jacket_rear_r_base, swing1LimitDegrees=12, swing2LimitDegrees=8, twistLimitDegrees=4}
 {bone1Name=dyn_jacket_rear_r_02, bone2Name=dyn_jacket_rear_r_01, swing1LimitDegrees=12, swing2LimitDegrees=8, twistLimitDegrees=4}
 {bone1Name=dyn_jacket_rear_r_03, bone2Name=dyn_jacket_rear_r_02, swing1LimitDegrees=12, swing2LimitDegrees=8, twistLimitDegrees=4}
 {bone1Name=dyn_jacket_rear_r_04, bone2Name=dyn_jacket_rear_r_03, swing1LimitDegrees=12, swing2LimitDegrees=8, twistLimitDegrees=4}
 {bone1Name=dyn_jacket_rear_r_05, bone2Name=dyn_jacket_rear_r_04, swing1LimitDegrees=12, swing2LimitDegrees=8, twistLimitDegrees=4}
 {bone1Name=dyn_jacket_rear_06, bone2Name=dyn_jacket_rear_05, swing1LimitDegrees=12, swing2LimitDegrees=8, twistLimitDegrees=4}
 {bone1Name=dyn_jacket_rear_l_06, bone2Name=dyn_jacket_rear_l_05, swing1LimitDegrees=12, swing2LimitDegrees=8, twistLimitDegrees=4}
 {bone1Name=dyn_jacket_rear_r_06, bone2Name=dyn_jacket_rear_r_05, swing1LimitDegrees=12, swing2LimitDegrees=8, twistLimitDegrees=4}
 {bone1Name=dyn_jacket_front_l_06, bone2Name=dyn_jacket_front_l_05, swing1LimitDegrees=24, swing2LimitDegrees=8, twistLimitDegrees=4}
 {bone1Name=dyn_jacket_front_l_07, bone2Name=dyn_jacket_front_l_06, swing1LimitDegrees=24, swing2LimitDegrees=8, twistLimitDegrees=4}
 {bone1Name=dyn_jacket_front_r_06, bone2Name=dyn_jacket_front_r_05, swing1LimitDegrees=12, swing2LimitDegrees=8, twistLimitDegrees=4}
 {bone1Name=dyn_jacket_front_r_07, bone2Name=dyn_jacket_front_r_06, swing1LimitDegrees=12, swing2LimitDegrees=8, twistLimitDegrees=4}
 {
  bone1Name=dyn_jacket_rear_l_04
  bone2Name=dyn_jacket_rear_04
  swing1LimitDegrees=45
  swing1Motion=Free
  swing2LimitDegrees=45
  swing2Motion=Free
  twistLimitDegrees=45
  twistMotion=Free
 }
 {
  bone1Name=dyn_jacket_rear_r_04
  bone2Name=dyn_jacket_rear_04
  swing1LimitDegrees=45
  swing1Motion=Free
  swing2LimitDegrees=45
  swing2Motion=Free
  twistLimitDegrees=45
  twistMotion=Free
 }
 {
  bone1Name=dyn_jacket_front_l_05
  bone2Name=dyn_jacket_rear_l_04
  swing1LimitDegrees=45
  swing1Motion=Free
  swing2LimitDegrees=45
  swing2Motion=Free
  twistLimitDegrees=45
  twistMotion=Free
 }
 {
  bone1Name=dyn_jacket_front_r_05
  bone2Name=dyn_jacket_rear_r_04
  swing1LimitDegrees=45
  swing1Motion=Free
  swing2LimitDegrees=45
  swing2Motion=Free
  twistLimitDegrees=45
  twistMotion=Free
 }
]
