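# Post-process Anim Blueprint for the F_MED_VowelSleep_Expand female body cosmetic, adding secondary motion to the pose copied from its parent mesh.
# Control Rig deformation feeds AnimDynamics chains for hood, wires, lapels and skirt, with backbling switching the hood to bone constraints.
# With cloth disabled, RigidBody simulation and jacket bone offsets are added.
# The event graph swaps skin-weight profiles, drives Chaos cloth, retunes for emotes, skydiving and frontend, and interpolates slope warp.
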AnimBlueprint /BRCosmetics/Characters/Player/Female/Medium/Bodies/F_MED_VowelSleep_Expand/Meshes/F_MED_VowelSleep_Expand_AnimBP
target_skeleton=&/F_MED_VowelSleep_Expand_Skeleton
graphs [
 AnimGraph {nodes [
   n1 = "Constraint - def_simspace (P)"!(ComponentPose=Local To Component(LocalPose=Use cached pose 'Post Extra Dynamics'()), bAlphaBoolEnabled:bool=True, ConstraintWeights_0=1) -> (Pose) {
    alphaInputType=Bool
    alphaBoolBlend {blendInTime=0.1, blendOutTime=0.1}
    boneToModify {boneName=def_simspace}
    constraintSetup [{targetBone {boneName=pelvis}}]
    constraintWeights [1]
   }
   n2 = RigidBody!(ComponentPose=@n1, Alpha, OverrideWorldGravity:Vector="0,0,-980", ExternalForce) -> (Pose) {
    bOverrideWorldGravity=true
    overridePhysicsAsset=&/F_MED_VowelSleep_Expand_Physics
    simulationSpace=BaseBoneSpace
    baseBoneRef {boneName=def_simspace}
    componentAppliedLinearAccClamp=<0 0 0>
    overrideWorldGravity {z=-980}
    simSpaceSettings {
     maxAngularAcceleration=3
     maxAngularVelocity=9
     maxLinearAcceleration=160
     maxLinearVelocity=1400
     velocityScaleZ=0.8
     worldAlpha=0.7
     externalLinearDragV=<0.1 0.1 0.1>
    }
   }
   n3 = "Transform (Modify) Bone - Bone: dyn_jacket_front_l_03"!(ComponentPose=@n2, bAlphaBoolEnabled:bool=True, Translation="0.026,-1.983,-0.783", Rotation="-3.456,0.204,-0.079", Scale="1,1,1") -> (Pose) {
    alphaInputType=Bool
    rotationMode=BMM_Additive
    rotationSpace=BCS_BoneSpace
    translationMode=BMM_Additive
    translationSpace=BCS_BoneSpace
    boneToModify {boneName=dyn_jacket_front_l_03}
    rotation=<rot -3.456 0.204 -0.079>
    translation=<0.026 -1.983 -0.783>
   }
   n4 = "Transform (Modify) Bone - Bone: dyn_jacket_front_l_04"!(ComponentPose=@n3, bAlphaBoolEnabled:bool=True, Translation="0,1,0", Rotation="0,0,0", Scale="1,1,1") -> (Pose) {
    alphaInputType=Bool
    rotationMode=BMM_Additive
    rotationSpace=BCS_BoneSpace
    translationMode=BMM_Additive
    translationSpace=BCS_BoneSpace
    boneToModify {boneName=dyn_jacket_front_l_04}
    translation {y=1}
   }
   n5 = "Transform (Modify) Bone - Bone: dyn_jacket_rear_l_03"!(ComponentPose=@n4, bAlphaBoolEnabled:bool=True, Translation="-0.005,0,2", Rotation="-0.001,-3,9.723", Scale="1,1,1") -> (Pose) {
    alphaInputType=Bool
    rotationMode=BMM_Additive
    rotationSpace=BCS_BoneSpace
    translationMode=BMM_Additive
    translationSpace=BCS_BoneSpace
    boneToModify {boneName=dyn_jacket_rear_l_03}
    rotation=<rot -0.001 -3 9.723>
    translation {x=-0.005, z=2}
   }
   Output Pose(Result=Blend Poses by bool(BlendPose_0=Use cached pose 'Post Extra Dynamics'(), BlendPose_1=Component To Local(ComponentPose=@n5), bActiveValue=Get bClothEnabled())) {layerGroup=None}
   n6 = Anim Dynamics - Bone: dyn_headphones(ComponentPose=Local To Component(LocalPose=Copy Pose From Mesh() {bUseAttachedParent=true}), Alpha=0.65, ExternalForce="0,0,0") {
    angularDampingOverride=0.999
    angularSpringConstant=100
    bAngularSpring=true
    bLinearSpring=true
    bOverrideAngularDamping=true
    bOverrideLinearDamping=true
    bUseGravityOverride=true
    gravityScale=0.35
    linearDampingOverride=0.999
    linearSpringConstant=100
    simulationSpace=RootRelative
    boundBone {boneName=dyn_headphones}
    componentAppliedLinearAccClamp=<0 0 0>
    gravityOverride {z=-1400}
    physicsBodyDefinitions [{
      boundBone {boneName=dyn_headphones}
      boxExtents=<20 20 20>
      constraintSetup {angularLimitsMax=<5 5 0>, angularLimitsMin=<-5 -5 -12>, angularTarget=<1 0 0>, linearAxesMax=<0 0.2 0>, linearAxesMin=<0 -1 0>}
      localJointOffset=<8 0 0>
     }]
    relativeSpaceBone {boneName=spine_04}
    simSpaceSettings {maxAngularAcceleration=40, maxAngularVelocity=900, simSpaceAngularAlpha=1}
   }
   n7 = "Transform (Modify) Bone - Bone: def_slope_warp"!(ComponentPose=@n6, bAlphaBoolEnabled:bool=True, Translation="0,0,0", Rotation="0,0,0", Scale="1,1,1") -> (Pose) {
    alphaInputType=Bool
    rotationMode=BMM_Replace
    rotationSpace=BCS_WorldSpace
    translationMode=BMM_Replace
    translationSpace=BCS_BoneSpace
    alphaBoolBlend {blendOption=Linear}
    boneToModify {boneName=def_slope_warp}
   }
   n8 = Control Rig!(Source=Component To Local(ComponentPose=@n7), Alpha, ClothEnabled:bool=Get bClothEnabled()) -> (Pose) {controlRigAssetReference {blueprintRigClass=&/F_MED_VowelSleep_Expand_CtrlRig.F_MED_VowelSleep_Expand_CtrlRig_C}}
   Save cached pose 'Post Deformation'(Pose=@n8) {cacheName=Post Deformation}
   n9 = Anim Dynamics - Bone: dyn_hood_01(ComponentPose=Local To Component(LocalPose=Use cached pose 'Post Deformation'()), Alpha=1, ExternalForce="0,0,0") {
    angularDampingOverride=0.999
    angularSpringConstant=20
    bAngularSpring=true
    bOverrideAngularDamping=true
    bUseGravityOverride=true
    gravityScale=0.35
    simulationSpace=RootRelative
    boundBone {boneName=dyn_hood_01}
    componentAppliedLinearAccClamp=<0 0 0>
    gravityOverride {z=-440}
    physicsBodyDefinitions [{
      boundBone {boneName=dyn_hood_01}
      boxExtents=<15 30 30>
      constraintSetup {angularLimitsMax=<5 5 0>, angularLimitsMin=<-5 0 -12>, angularTarget=<1 0 0>}
      localJointOffset=<8 0 0>
     }]
    simSpaceSettings {maxAngularAcceleration=4, maxAngularVelocity=400, simSpaceAngularAlpha=1}
   }
   n10 = Anim Dynamics - Bone: dyn_hood_02(ComponentPose=@n9, Alpha=1, ExternalForce="0,0,0") {
    angularDampingOverride=0.999
    angularSpringConstant=20
    bAngularSpring=true
    bOverrideAngularDamping=true
    bUseGravityOverride=true
    gravityScale=0.35
    simulationSpace=RootRelative
    boundBone {boneName=dyn_hood_02}
    componentAppliedLinearAccClamp=<0 0 0>
    gravityOverride {z=-440}
    physicsBodyDefinitions [{
      boundBone {boneName=dyn_hood_02}
      boxExtents=<15 30 30>
      constraintSetup {angularLimitsMax=<5 5 0>, angularLimitsMin=<-5 0 -28>, angularTarget=<1 0 0>}
      localJointOffset=<8 0 0>
     }]
    simSpaceSettings {maxAngularAcceleration=4, maxAngularVelocity=400, simSpaceAngularAlpha=1}
   }
   n11 = Anim Dynamics - Bone: dyn_hood_03(ComponentPose=@n10, Alpha=1, ExternalForce="0,0,0") {
    angularDampingOverride=0.999
    angularSpringConstant=20
    bAngularSpring=true
    bOverrideAngularDamping=true
    bUseGravityOverride=true
    gravityScale=0.35
    simulationSpace=RootRelative
    boundBone {boneName=dyn_hood_03}
    componentAppliedLinearAccClamp=<0 0 0>
    gravityOverride {z=-440}
    physicsBodyDefinitions [{
      boundBone {boneName=dyn_hood_03}
      boxExtents=<15 30 30>
      constraintSetup {angularLimitsMax=<5 5 0>, angularLimitsMin=<-5 0 -48>, angularTarget=<1 0 0>}
      localJointOffset=<8 0 0>
     }]
    simSpaceSettings {maxAngularAcceleration=4, maxAngularVelocity=400, simSpaceAngularAlpha=1}
   }
   n12 = "Constraint - dyn_hood_01 (PPP)"!(
    ComponentPose=Local To Component(LocalPose=Use cached pose 'Post Deformation'())
    Alpha=1
    ConstraintWeights_0=0.6701
    ConstraintWeights_1:float=0.1666
    ConstraintWeights_2:float=0.1633
   ) -> (Pose) {
    boneToModify {boneName=dyn_hood_01}
    constraintSetup [{targetBone {boneName=spine_05}}, {targetBone {boneName=clavicle_r}}, {targetBone {boneName=clavicle_l}}]
    constraintWeights [1,1,1]
   }
   n13 = "Constraint - dyn_hood_02 (PPPPPP)"!(
    ComponentPose=@n12
    Alpha=1
    ConstraintWeights_0=0.4644
    ConstraintWeights_1:float=0.2482
    ConstraintWeights_2:float=0.1309
    ConstraintWeights_3:float=0.1298
    ConstraintWeights_4:float=0.0146
    ConstraintWeights_5:float=0.012
   ) -> (Pose) {
    boneToModify {boneName=dyn_hood_02}
    constraintSetup [
     {targetBone {boneName=spine_05}}
     {targetBone {boneName=spine_04}}
     {targetBone {boneName=clavicle_l}}
     {targetBone {boneName=clavicle_r}}
     {targetBone {boneName=spine_03}}
     {targetBone {boneName=spine_02}}
    ]
    constraintWeights [1,1,1,1,1,1]
   }
   n14 = "Constraint - dyn_hood_03 (PPPPPPP)"!(
    ComponentPose=@n13
    Alpha=1
    ConstraintWeights_0=0.3571
    ConstraintWeights_1:float=0.2442
    ConstraintWeights_2:float=0.0937
    ConstraintWeights_3:float=0.092
    ConstraintWeights_4:float=0.0918
    ConstraintWeights_5:float=0.0912
    ConstraintWeights_6:float=0.03
   ) -> (Pose) {
    boneToModify {boneName=dyn_hood_03}
    constraintSetup [
     {targetBone {boneName=spine_04}}
     {targetBone {boneName=spine_05}}
     {targetBone {boneName=spine_02}}
     {targetBone {boneName=clavicle_l}}
     {targetBone {boneName=clavicle_r}}
     {targetBone {boneName=spine_03}}
     {targetBone {boneName=spine_01}}
    ]
    constraintWeights [1,1,1,1,1,1,1]
   }
   Save cached pose 'Post Hood Dynamics'(Pose=Blend Poses by bool(
     BlendPose_0=Component To Local(ComponentPose=@n14)
     BlendPose_1=Component To Local(ComponentPose=@n11)
     bActiveValue=Get isWearingBackbling()
    )) {cacheName=Post Hood Dynamics}
   n15 = Anim Dynamics - Bone: dyn_headphone_wire_01_l(ComponentPose=Local To Component(LocalPose=Use cached pose 'Post Hood Dynamics'()), Alpha=0.7, ExternalForce="0,0,0") {
    angularDampingOverride=0.999
    angularSpringConstant=200
    bAngularSpring=true
    bLinearSpring=true
    bOverrideAngularDamping=true
    bOverrideLinearDamping=true
    bUseGravityOverride=true
    gravityScale=0.35
    linearDampingOverride=0.999
    linearSpringConstant=100
    simulationSpace=RootRelative
    boundBone {boneName=dyn_headphone_wire_01_l}
    componentLinearAccScale=<1 1 1>
    componentLinearVelScale=<1 1 1>
    gravityOverride {z=-980}
    physicsBodyDefinitions [{
      boundBone {boneName=dyn_headphone_wire_01_l}
      boxExtents=<20 10 10>
      constraintSetup {angularLimitsMax=<15 15 15>, angularLimitsMin=<-15 -15 -15>, angularTarget=<1 0 0>, linearAxesMax=<0 0 0.8>, linearAxesMin=<-0.5 0 -0.8>}
      localJointOffset=<4 0 0>
     }]
   }
   n16 = Anim Dynamics - Bone: dyn_headphone_wire_01_r(ComponentPose=@n15, Alpha=0.7, ExternalForce="0,0,0") {
    angularDampingOverride=0.999
    angularSpringConstant=200
    bAngularSpring=true
    bLinearSpring=true
    bOverrideAngularDamping=true
    bOverrideLinearDamping=true
    bUseGravityOverride=true
    gravityScale=0.35
    linearDampingOverride=0.999
    linearSpringConstant=100
    simulationSpace=RootRelative
    boundBone {boneName=dyn_headphone_wire_01_r}
    componentLinearAccScale=<1 1 1>
    componentLinearVelScale=<1 1 1>
    gravityOverride {z=-980}
    physicsBodyDefinitions [{
      boundBone {boneName=dyn_headphone_wire_01_r}
      boxExtents=<20 10 10>
      constraintSetup {angularLimitsMax=<15 15 15>, angularLimitsMin=<-15 -15 -15>, angularTarget=<1 0 0>, linearAxesMax=<1 0 0.5>, linearAxesMin=<-1 0 -1>}
      localJointOffset=<4 0 0>
     }]
   }
   n17 = Anim Dynamics - Bone: dyn_headphone_wire_02_l(ComponentPose=@n16, Alpha=0.7, ExternalForce="0,0,0") {
    angularDampingOverride=0.999
    angularSpringConstant=200
    bAngularSpring=true
    bLinearSpring=true
    bOverrideAngularDamping=true
    bOverrideLinearDamping=true
    bUseGravityOverride=true
    gravityScale=0.35
    linearDampingOverride=0.999
    linearSpringConstant=100
    simulationSpace=RootRelative
    boundBone {boneName=dyn_headphone_wire_02_l}
    componentLinearAccScale=<1 1 1>
    componentLinearVelScale=<1 1 1>
    gravityOverride {z=-980}
    physicsBodyDefinitions [{
      boundBone {boneName=dyn_headphone_wire_02_l}
      boxExtents=<20 10 10>
      constraintSetup {angularLimitsMax=<15 15 15>, angularLimitsMin=<-15 -15 -15>, angularTarget=<1 0 0>, linearAxesMax=<0.5 0 1>, linearAxesMin=<-0.5 0 -1>}
      localJointOffset=<4 0 0>
     }]
   }
   n18 = Anim Dynamics - Bone: dyn_headphone_wire_02_r(ComponentPose=@n17, Alpha=0.7, ExternalForce="0,0,0") {
    angularDampingOverride=0.999
    angularSpringConstant=200
    bAngularSpring=true
    bLinearSpring=true
    bOverrideAngularDamping=true
    bOverrideLinearDamping=true
    bUseGravityOverride=true
    gravityScale=0.35
    linearDampingOverride=0.999
    linearSpringConstant=100
    simulationSpace=RootRelative
    boundBone {boneName=dyn_headphone_wire_02_r}
    componentLinearAccScale=<1 1 1>
    componentLinearVelScale=<1 1 1>
    gravityOverride {z=-980}
    physicsBodyDefinitions [{
      boundBone {boneName=dyn_headphone_wire_02_r}
      boxExtents=<20 10 10>
      constraintSetup {angularLimitsMax=<15 15 15>, angularLimitsMin=<-15 -15 -15>, angularTarget=<1 0 0>, linearAxesMax=<0.5 0 1>, linearAxesMin=<-0.5 0 -1>}
      localJointOffset=<4 0 0>
     }]
   }
   n19 = Anim Dynamics - Bone: dyn_headphone_wire_03_c(ComponentPose=@n18, Alpha=1, ExternalForce="0,0,0") {
    angularDampingOverride=0.999
    angularSpringConstant=200
    bAngularSpring=true
    bLinearSpring=true
    bOverrideAngularDamping=true
    bOverrideLinearDamping=true
    bUseGravityOverride=true
    gravityScale=0.35
    linearDampingOverride=0.999
    linearSpringConstant=100
    simulationSpace=RootRelative
    boundBone {boneName=dyn_headphone_wire_03_c}
    componentLinearAccScale=<1 1 1>
    componentLinearVelScale=<1 1 1>
    gravityOverride {z=-980}
    physicsBodyDefinitions [{
      boundBone {boneName=dyn_headphone_wire_03_c}
      boxExtents=<20 10 10>
      constraintSetup {angularLimitsMax=<15 15 15>, angularLimitsMin=<-15 -15 -15>, angularTarget=<1 0 0>, linearAxesMax=<0.5 0 0.5>, linearAxesMin=<-0.5 0 -0.5>}
      localJointOffset=<4 0 0>
     }]
   }
   n20 = Anim Dynamics - Bone: dyn_headphone_wire_04_c(ComponentPose=@n19, Alpha=1, ExternalForce="0,0,0") {
    angularDampingOverride=0.999
    angularSpringConstant=200
    bAngularSpring=true
    bLinearSpring=true
    bOverrideAngularDamping=true
    bOverrideLinearDamping=true
    bUseGravityOverride=true
    gravityScale=0.35
    linearDampingOverride=0.999
    linearSpringConstant=100
    simulationSpace=RootRelative
    boundBone {boneName=dyn_headphone_wire_04_c}
    componentLinearAccScale=<1 1 1>
    componentLinearVelScale=<1 1 1>
    gravityOverride {z=-980}
    physicsBodyDefinitions [{
      boundBone {boneName=dyn_headphone_wire_04_c}
      boxExtents=<20 10 10>
      constraintSetup {angularLimitsMax=<15 15 15>, angularLimitsMin=<-15 -15 -15>, angularTarget=<1 0 0>, linearAxesMax=<0.5 0 0.5>, linearAxesMin=<-0.5 0 -0.5>}
      localJointOffset=<4 0 0>
     }]
   }
   n21 = Anim Dynamics - Bone: dyn_lapel_l_01(ComponentPose=@n20, Alpha=1, ExternalForce="0,0,0") {
    angularDampingOverride=1
    angularSpringConstant=20000
    bAngularSpring=true
    bOverrideAngularDamping=true
    bUseGravityOverride=true
    gravityScale=0.35
    simulationSpace=RootRelative
    boundBone {boneName=dyn_lapel_l_01}
    componentAppliedLinearAccClamp=<0 0 0>
    gravityOverride {z=-980}
    physicsBodyDefinitions [{
      boundBone {boneName=dyn_lapel_l_01}
      boxExtents=<30 15 30>
      constraintSetup {angularLimitsMax=<0 5 0>, angularLimitsMin=<0 -5 -24>, angularTarget=<1 0 0>}
      localJointOffset=<4 0 0>
     }]
    simSpaceSettings {maxAngularAcceleration=30, maxAngularVelocity=300, simSpaceAngularAlpha=1}
   }
   n22 = Anim Dynamics - Bone: dyn_lapel_l_02(ComponentPose=@n21, Alpha=1, ExternalForce="0,0,0") {
    angularDampingOverride=1
    angularSpringConstant=20000
    bAngularSpring=true
    bOverrideAngularDamping=true
    bUseGravityOverride=true
    gravityScale=0.35
    simulationSpace=RootRelative
    boundBone {boneName=dyn_lapel_l_02}
    componentAppliedLinearAccClamp=<0 0 0>
    gravityOverride {z=-980}
    physicsBodyDefinitions [{
      boundBone {boneName=dyn_lapel_l_02}
      boxExtents=<30 15 30>
      constraintSetup {angularLimitsMax=<0 5 0>, angularLimitsMin=<0 -5 -24>, angularTarget=<1 0 0>}
      localJointOffset=<4 0 0>
     }]
    simSpaceSettings {maxAngularAcceleration=30, maxAngularVelocity=300, simSpaceAngularAlpha=1}
   }
   n23 = Anim Dynamics - Bone: dyn_lapel_r_01(ComponentPose=@n22, Alpha=1, ExternalForce="0,0,0") {
    angularDampingOverride=1
    angularSpringConstant=20000
    bAngularSpring=true
    bOverrideAngularDamping=true
    bUseGravityOverride=true
    gravityScale=0.35
    simulationSpace=RootRelative
    boundBone {boneName=dyn_lapel_r_01}
    componentAppliedLinearAccClamp=<0 0 0>
    gravityOverride {z=-980}
    physicsBodyDefinitions [{
      boundBone {boneName=dyn_lapel_r_01}
      boxExtents=<30 15 30>
      constraintSetup {angularLimitsMax=<0 1 0>, angularLimitsMin=<0 -1 -24>, angularTarget=<1 0 0>}
      localJointOffset=<-4 0 0>
     }]
    simSpaceSettings {maxAngularAcceleration=30, maxAngularVelocity=300, simSpaceAngularAlpha=1}
   }
   n24 = Anim Dynamics - Bone: dyn_lapel_r_02(ComponentPose=@n23, Alpha=1, ExternalForce="0,0,0") {
    angularDampingOverride=1
    angularSpringConstant=20000
    bAngularSpring=true
    bOverrideAngularDamping=true
    bUseGravityOverride=true
    gravityScale=0.35
    simulationSpace=RootRelative
    boundBone {boneName=dyn_lapel_r_02}
    componentAppliedLinearAccClamp=<0 0 0>
    gravityOverride {z=-980}
    physicsBodyDefinitions [{
      boundBone {boneName=dyn_lapel_r_02}
      boxExtents=<30 15 30>
      constraintSetup {angularLimitsMax=<0 2 0>, angularLimitsMin=<0 -2 -24>, angularTarget=<1 0 0>}
      localJointOffset=<-4 0 0>
     }]
    simSpaceSettings {maxAngularAcceleration=30, maxAngularVelocity=300, simSpaceAngularAlpha=1}
   }
   n25 = Anim Dynamics - Bone: dyn_skirt_front_01!(ComponentPose=@n24, Alpha=1, GravityOverride:Vector="0,0,-1800", ExternalForce="0,0,0") -> (Pose) {
    angularDampingOverride=0.999
    angularSpringConstant=20
    bAngularSpring=true
    bOverrideAngularDamping=true
    bUseGravityOverride=true
    gravityScale=0.35
    simulationSpace=RootRelative
    boundBone {boneName=dyn_skirt_front_01}
    componentAppliedLinearAccClamp=<0 0 0>
    gravityOverride {z=-1800}
    physicsBodyDefinitions [{
      boundBone {boneName=dyn_skirt_front_01}
      boxExtents=<15 15 30>
      constraintSetup {angularLimitsMax=<4 4 0>, angularLimitsMin=<-4 -4 -16>, angularTarget=<1 0 0>}
      localJointOffset=<4 0 0>
     }]
    simSpaceSettings {maxAngularAcceleration=4, maxAngularVelocity=400, simSpaceAngularAlpha=0.7}
   }
   n26 = Anim Dynamics - Bone: dyn_skirt_front_02!(ComponentPose=@n25, Alpha=1, GravityOverride:Vector="0,0,-1800", ExternalForce="0,0,0") -> (Pose) {
    angularDampingOverride=0.999
    angularSpringConstant=40
    bAngularSpring=true
    bOverrideAngularDamping=true
    bUseGravityOverride=true
    gravityScale=0.35
    simulationSpace=RootRelative
    boundBone {boneName=dyn_skirt_front_02}
    componentAppliedLinearAccClamp=<0 0 0>
    gravityOverride {z=-1800}
    physicsBodyDefinitions [{
      boundBone {boneName=dyn_skirt_front_02}
      boxExtents=<15 15 30>
      constraintSetup {angularLimitsMax=<4 8 0>, angularLimitsMin=<-4 -8 -24>, angularTarget=<1 0 0>}
      localJointOffset=<4 0 0>
     }]
    simSpaceSettings {maxAngularAcceleration=4, maxAngularVelocity=400, simSpaceAngularAlpha=0.7}
   }
   n27 = Anim Dynamics - Bone: dyn_skirt_front_l_01!(ComponentPose=@n26, Alpha=1, GravityOverride:Vector="0,0,-1800", ExternalForce="0,0,0") -> (Pose) {
    angularDampingOverride=0.999
    angularSpringConstant=20
    bAngularSpring=true
    bOverrideAngularDamping=true
    bUseGravityOverride=true
    gravityScale=0.35
    simulationSpace=RootRelative
    boundBone {boneName=dyn_skirt_front_l_01}
    componentAppliedLinearAccClamp=<0 0 0>
    gravityOverride {z=-1800}
    physicsBodyDefinitions [{
      boundBone {boneName=dyn_skirt_front_l_01}
      boxExtents=<15 15 30>
      constraintSetup {angularLimitsMax=<4 4 0>, angularLimitsMin=<-4 -4 -16>, angularTarget=<1 0 0>}
      localJointOffset=<4 0 0>
     }]
    simSpaceSettings {maxAngularAcceleration=4, maxAngularVelocity=400, simSpaceAngularAlpha=0.7}
   }
   n28 = Anim Dynamics - Bone: dyn_skirt_front_l_02!(ComponentPose=@n27, Alpha=1, GravityOverride:Vector="0,0,-1800", ExternalForce="0,0,0") -> (Pose) {
    angularDampingOverride=0.999
    angularSpringConstant=40
    bAngularSpring=true
    bOverrideAngularDamping=true
    bUseGravityOverride=true
    gravityScale=0.35
    simulationSpace=RootRelative
    boundBone {boneName=dyn_skirt_front_l_02}
    componentAppliedLinearAccClamp=<0 0 0>
    gravityOverride {z=-1800}
    physicsBodyDefinitions [{
      boundBone {boneName=dyn_skirt_front_l_02}
      boxExtents=<15 15 30>
      constraintSetup {angularLimitsMax=<4 8 0>, angularLimitsMin=<-4 -8 -24>, angularTarget=<1 0 0>}
      localJointOffset=<4 0 0>
     }]
    simSpaceSettings {maxAngularAcceleration=4, maxAngularVelocity=400, simSpaceAngularAlpha=0.7}
   }
   n29 = Anim Dynamics - Bone: dyn_skirt_side_l_01!(ComponentPose=@n28, Alpha=1, GravityOverride:Vector="0,0,-1800", ExternalForce="0,0,0") -> (Pose) {
    angularDampingOverride=0.999
    angularSpringConstant=20
    bAngularSpring=true
    bOverrideAngularDamping=true
    bUseGravityOverride=true
    gravityScale=0.35
    simulationSpace=RootRelative
    boundBone {boneName=dyn_skirt_side_l_01}
    componentAppliedLinearAccClamp=<0 0 0>
    gravityOverride {z=-1800}
    physicsBodyDefinitions [{
      boundBone {boneName=dyn_skirt_side_l_01}
      boxExtents=<15 15 30>
      constraintSetup {angularLimitsMax=<2 2 0>, angularLimitsMin=<-2 -2 -6>, angularTarget=<1 0 0>}
      localJointOffset=<4 0 0>
     }]
    simSpaceSettings {maxAngularAcceleration=4, maxAngularVelocity=400, simSpaceAngularAlpha=0.7}
   }
   n30 = Anim Dynamics - Bone: dyn_skirt_side_l_02!(ComponentPose=@n29, Alpha=1, GravityOverride:Vector="0,0,-1800", ExternalForce="0,0,0") -> (Pose) {
    angularDampingOverride=0.999
    angularSpringConstant=40
    bAngularSpring=true
    bOverrideAngularDamping=true
    bUseGravityOverride=true
    gravityScale=0.35
    simulationSpace=RootRelative
    boundBone {boneName=dyn_skirt_side_l_02}
    componentAppliedLinearAccClamp=<0 0 0>
    gravityOverride {z=-1800}
    physicsBodyDefinitions [{
      boundBone {boneName=dyn_skirt_side_l_02}
      boxExtents=<15 15 30>
      constraintSetup {angularLimitsMax=<2 2 0>, angularLimitsMin=<-2 -2 -6>, angularTarget=<1 0 0>}
      localJointOffset=<4 0 0>
     }]
    simSpaceSettings {maxAngularAcceleration=4, maxAngularVelocity=400, simSpaceAngularAlpha=0.7}
   }
   n31 = Anim Dynamics - Bone: dyn_skirt_rear_l_01!(ComponentPose=@n30, Alpha=1, GravityOverride:Vector="0,0,-1800", ExternalForce="0,0,0") -> (Pose) {
    angularDampingOverride=0.999
    angularSpringConstant=20
    bAngularSpring=true
    bOverrideAngularDamping=true
    bUseGravityOverride=true
    gravityScale=0.35
    simulationSpace=RootRelative
    boundBone {boneName=dyn_skirt_rear_l_01}
    componentAppliedLinearAccClamp=<0 0 0>
    gravityOverride {z=-1800}
    physicsBodyDefinitions [{
      boundBone {boneName=dyn_skirt_rear_l_01}
      boxExtents=<15 15 30>
      constraintSetup {angularLimitsMax=<2 2 0>, angularLimitsMin=<-2 -2 -6>, angularTarget=<1 0 0>}
      localJointOffset=<4 0 0>
     }]
    simSpaceSettings {maxAngularAcceleration=4, maxAngularVelocity=400, simSpaceAngularAlpha=0.7}
   }
   n32 = Anim Dynamics - Bone: dyn_skirt_rear_l_02!(ComponentPose=@n31, Alpha=1, GravityOverride:Vector="0,0,-1800", ExternalForce="0,0,0") -> (Pose) {
    angularDampingOverride=0.999
    angularSpringConstant=40
    bAngularSpring=true
    bOverrideAngularDamping=true
    bUseGravityOverride=true
    gravityScale=0.35
    simulationSpace=RootRelative
    boundBone {boneName=dyn_skirt_rear_l_02}
    componentAppliedLinearAccClamp=<0 0 0>
    gravityOverride {z=-1800}
    physicsBodyDefinitions [{
      boundBone {boneName=dyn_skirt_rear_l_02}
      boxExtents=<15 15 30>
      constraintSetup {angularLimitsMax=<2 2 0>, angularLimitsMin=<-2 -2 -6>, angularTarget=<1 0 0>}
      localJointOffset=<4 0 0>
     }]
    simSpaceSettings {maxAngularAcceleration=4, maxAngularVelocity=400, simSpaceAngularAlpha=0.7}
   }
   n33 = Anim Dynamics - Bone: dyn_skirt_rear_01!(ComponentPose=@n32, Alpha=1, GravityOverride:Vector="0,0,-1800", ExternalForce="0,0,0") -> (Pose) {
    angularDampingOverride=0.999
    angularSpringConstant=20
    bAngularSpring=true
    bOverrideAngularDamping=true
    bUseGravityOverride=true
    gravityScale=0.35
    simulationSpace=RootRelative
    boundBone {boneName=dyn_skirt_rear_01}
    componentAppliedLinearAccClamp=<0 0 0>
    gravityOverride {z=-1800}
    physicsBodyDefinitions [{
      boundBone {boneName=dyn_skirt_rear_01}
      boxExtents=<15 15 30>
      constraintSetup {angularLimitsMax=<2 2 0>, angularLimitsMin=<-2 -2 -6>, angularTarget=<1 0 0>}
      localJointOffset=<4 0 0>
     }]
    simSpaceSettings {maxAngularAcceleration=4, maxAngularVelocity=400, simSpaceAngularAlpha=0.7}
   }
   n34 = Anim Dynamics - Bone: dyn_skirt_rear_02!(ComponentPose=@n33, Alpha=1, GravityOverride:Vector="0,0,-1800", ExternalForce="0,0,0") -> (Pose) {
    angularDampingOverride=0.999
    angularSpringConstant=40
    bAngularSpring=true
    bOverrideAngularDamping=true
    bUseGravityOverride=true
    gravityScale=0.35
    simulationSpace=RootRelative
    boundBone {boneName=dyn_skirt_rear_02}
    componentAppliedLinearAccClamp=<0 0 0>
    gravityOverride {z=-1800}
    physicsBodyDefinitions [{
      boundBone {boneName=dyn_skirt_rear_02}
      boxExtents=<15 15 30>
      constraintSetup {angularLimitsMax=<2 2 0>, angularLimitsMin=<-2 -2 -6>, angularTarget=<1 0 0>}
      localJointOffset=<4 0 0>
     }]
    simSpaceSettings {maxAngularAcceleration=4, maxAngularVelocity=400, simSpaceAngularAlpha=0.7}
   }
   n35 = Anim Dynamics - Bone: dyn_skirt_front_r_01!(ComponentPose=@n34, Alpha=1, GravityOverride:Vector="0,0,-1800", ExternalForce="0,0,0") -> (Pose) {
    angularDampingOverride=0.999
    angularSpringConstant=20
    bAngularSpring=true
    bOverrideAngularDamping=true
    bUseGravityOverride=true
    gravityScale=0.35
    simulationSpace=RootRelative
    boundBone {boneName=dyn_skirt_front_r_01}
    componentAppliedLinearAccClamp=<0 0 0>
    gravityOverride {z=-1800}
    physicsBodyDefinitions [{
      boundBone {boneName=dyn_skirt_front_r_01}
      boxExtents=<15 15 30>
      constraintSetup {angularLimitsMax=<4 4 0>, angularLimitsMin=<-4 -4 -16>, angularTarget=<1 0 0>}
      localJointOffset=<-4 0 0>
     }]
    simSpaceSettings {maxAngularAcceleration=4, maxAngularVelocity=400, simSpaceAngularAlpha=0.7}
   }
   n36 = Anim Dynamics - Bone: dyn_skirt_front_r_02!(ComponentPose=@n35, Alpha=1, GravityOverride:Vector="0,0,-1800", ExternalForce="0,0,0") -> (Pose) {
    angularDampingOverride=0.999
    angularSpringConstant=40
    bAngularSpring=true
    bOverrideAngularDamping=true
    bUseGravityOverride=true
    gravityScale=0.35
    simulationSpace=RootRelative
    boundBone {boneName=dyn_skirt_front_r_02}
    componentAppliedLinearAccClamp=<0 0 0>
    gravityOverride {z=-1800}
    physicsBodyDefinitions [{
      boundBone {boneName=dyn_skirt_front_r_02}
      boxExtents=<15 15 30>
      constraintSetup {angularLimitsMax=<4 8 0>, angularLimitsMin=<-4 -8 -24>, angularTarget=<1 0 0>}
      localJointOffset=<-4 0 0>
     }]
    simSpaceSettings {maxAngularAcceleration=4, maxAngularVelocity=400, simSpaceAngularAlpha=0.7}
   }
   n37 = Anim Dynamics - Bone: dyn_skirt_side_r_01!(ComponentPose=@n36, Alpha=1, GravityOverride:Vector="0,0,-1800", ExternalForce="0,0,0") -> (Pose) {
    angularDampingOverride=0.999
    angularSpringConstant=20
    bAngularSpring=true
    bOverrideAngularDamping=true
    bUseGravityOverride=true
    gravityScale=0.35
    simulationSpace=RootRelative
    boundBone {boneName=dyn_skirt_side_r_01}
    componentAppliedLinearAccClamp=<0 0 0>
    gravityOverride {z=-1800}
    physicsBodyDefinitions [{
      boundBone {boneName=dyn_skirt_side_r_01}
      boxExtents=<15 15 30>
      constraintSetup {angularLimitsMax=<2 2 0>, angularLimitsMin=<-2 -2 -6>, angularTarget=<1 0 0>}
      localJointOffset=<-4 0 0>
     }]
    simSpaceSettings {maxAngularAcceleration=4, maxAngularVelocity=400, simSpaceAngularAlpha=0.7}
   }
   n38 = Anim Dynamics - Bone: dyn_skirt_side_r_02!(ComponentPose=@n37, Alpha=1, GravityOverride:Vector="0,0,-1800", ExternalForce="0,0,0") -> (Pose) {
    angularDampingOverride=0.999
    angularSpringConstant=40
    bAngularSpring=true
    bOverrideAngularDamping=true
    bUseGravityOverride=true
    gravityScale=0.35
    simulationSpace=RootRelative
    boundBone {boneName=dyn_skirt_side_r_02}
    componentAppliedLinearAccClamp=<0 0 0>
    gravityOverride {z=-1800}
    physicsBodyDefinitions [{
      boundBone {boneName=dyn_skirt_side_r_02}
      boxExtents=<15 15 30>
      constraintSetup {angularLimitsMax=<2 2 0>, angularLimitsMin=<-2 -2 -6>, angularTarget=<1 0 0>}
      localJointOffset=<-4 0 0>
     }]
    simSpaceSettings {maxAngularAcceleration=4, maxAngularVelocity=400, simSpaceAngularAlpha=0.7}
   }
   n39 = Anim Dynamics - Bone: dyn_skirt_rear_r_01!(ComponentPose=@n38, Alpha=1, GravityOverride:Vector="0,0,-1800", ExternalForce="0,0,0") -> (Pose) {
    angularDampingOverride=0.999
    angularSpringConstant=20
    bAngularSpring=true
    bOverrideAngularDamping=true
    bUseGravityOverride=true
    gravityScale=0.35
    simulationSpace=RootRelative
    boundBone {boneName=dyn_skirt_rear_r_01}
    componentAppliedLinearAccClamp=<0 0 0>
    gravityOverride {z=-1800}
    physicsBodyDefinitions [{
      boundBone {boneName=dyn_skirt_rear_r_01}
      boxExtents=<15 15 30>
      constraintSetup {angularLimitsMax=<2 2 0>, angularLimitsMin=<-2 -2 -6>, angularTarget=<1 0 0>}
      localJointOffset=<-4 0 0>
     }]
    simSpaceSettings {maxAngularAcceleration=4, maxAngularVelocity=400, simSpaceAngularAlpha=0.7}
   }
   n40 = Anim Dynamics - Bone: dyn_skirt_rear_r_02!(ComponentPose=@n39, Alpha=1, GravityOverride:Vector="0,0,-1800", ExternalForce="0,0,0") -> (Pose) {
    angularDampingOverride=0.999
    angularSpringConstant=40
    bAngularSpring=true
    bOverrideAngularDamping=true
    bUseGravityOverride=true
    gravityScale=0.35
    simulationSpace=RootRelative
    boundBone {boneName=dyn_skirt_rear_r_02}
    componentAppliedLinearAccClamp=<0 0 0>
    gravityOverride {z=-1800}
    physicsBodyDefinitions [{
      boundBone {boneName=dyn_skirt_rear_r_02}
      boxExtents=<15 15 30>
      constraintSetup {angularLimitsMax=<2 2 0>, angularLimitsMin=<-2 -2 -6>, angularTarget=<1 0 0>}
      localJointOffset=<-4 0 0>
     }]
    simSpaceSettings {maxAngularAcceleration=4, maxAngularVelocity=400, simSpaceAngularAlpha=0.7}
   }
   Save cached pose 'Post Extra Dynamics'(Pose=Component To Local(ComponentPose=@n40)) {cacheName=Post Extra Dynamics}
   Use cached pose 'Post Deformation'()
  ]}
 EventGraph {nodes [
   n1 = Set Cloth Weight Profile!() -> (OutputDelegate, then, Cloth State:bool=~) {is=title K2Node_CustomEvent}
   > Is Valid(InputObject=Get Custom Base Mesh() -> (Custom Base Mesh:Skeletal Mesh Component Object Reference))
   > n2 = Sequence()
   > n3 = Branch(Condition=@n1.Cloth State)
   n4 = Get Custom Base Mesh()
   > SetSkinWeightProfile!(execute, self:Skinned Mesh Component Object Reference=@n4, InProfileName:Name=ClothWeighting, InLayer:ESkinWeightProfileLayer Enum=Secondary) -> (then, ReturnValue:bool)
   SetSkinWeightProfile(execute=@n3.else, self=@n4, InProfileName=Default, InLayer=Secondary)
   Set Prev Cloth Bool(execute=@n2.then_1, Prev Cloth Bool:bool=Get bClothEnabled()) -> (Output_Get:bool)
   n5 = Event BlueprintUpdateAnimation()
   > Set Delta Time X(Delta Time X=@n5.DeltaTimeX)
   > n6 = Sequence!(execute) -> (then_0, then_1, then_2:Exec, then_3:Exec)
   n7 = Branch(execute=@n6.then_2, Condition=Get bClothEnabled())
   > n8 = "Cloth Controller Function (Parameter Asset) (Nativized)"!(
    execute
    CharacterPartAnimInstance:Custom Character Part Anim Instance Object Reference=Self-Reference()
    ClothingAssetName:Name=F_MED_VowelSleep_Expand_Clothing
    ClothCoreSettings:"Cloth Core Settings Structure (by ref)"=Get Cloth Core Settings Runtime() -> (Cloth Core Settings Runtime:Cloth Core Settings Structure)
    WindGustLocalSettings:"Wind Gust Settings Structure (by ref)"=Get Wind Gust Local Settings() -> (Wind Gust Local Settings:Wind Gust Settings Structure)
    ClothUpAxis:Vector="0,0,1"
    LinVelScaleFwdBwdClamp:float=0
    LinVelScaleUpDwnClamp:float=0.7
    DeltaTime:float=Get Delta Time X()
    GameTimeInSeconds:float=GetGameTimeInSeconds()
    StateLogicLayerName:Name=none
    Parameters:Fort Character Dynamics Parameters Object Reference=&/F_MED_VowelSleep_Expand_Parameters
   ) -> (
    then
    OutCurrentState:Name=None
    OutClothParameters:Cloth Parameters Structure
    OutClothCoreSettings:Cloth Core Settings Structure
    OutWindGustSettings:Wind Gust Parameters Structure
    OutWindGustLocalSettings:Wind Gust Settings Structure
    OutWindVelocity:Vector
    OutLinearVelVector:Vector
   )
   > Set Wind Gust Local Settings(Wind Gust Local Settings:Wind Gust Settings Structure=@n8.OutWindGustLocalSettings) -> (Output_Get:Wind Gust Settings Structure)
   n9 = "Gravity Override State Function (Parameter Asset) (Nativized)"(
    execute=@n7.else
    CharacterPartAnimInstance=Self-Reference()
    GameTimeInSeconds=GetGameTimeInSeconds()
    Parameters=&F_MED_VowelSleep_Expand_Parameters
   )
   > Set Gravity Override(Gravity Override=@n9)
   Branch(
    execute=@n6.then_0
    Condition="Not Equal (Boolean)"(A=Get bClothEnabled(), B=Get Prev Cloth Bool() -> (Prev Cloth Bool:bool)) -> (ReturnValue=~)
   )
   > Set Cloth Weight Profile!(execute, self, Cloth State:bool=Get bClothEnabled()) -> (then) {is=title K2Node_CallFunction}
   Handle Slope Warp!(execute=@n6.then_3, self, DeltaTime:double=Get Delta Time X()) -> (then)
   Branch(execute=@n6.then_1, Condition=Get bIsInFrontEnd())
   > Set Cloth Core Settings Runtime(Cloth Core Settings Runtime:Cloth Core Settings Structure=Get Cloth Core Settings Frontend() -> (Cloth Core Settings Frontend:Cloth Core Settings Structure)) -> (Output_Get:Cloth Core Settings Structure)
   > Set Anim Dynamics Alpha(Anim Dynamics Alpha:double=0.35) -> (Output_Get:double)
   n10 = Event OnCharacterCustomizationLoaded()
   > Set Custom Base Mesh(Custom Base Mesh:Skeletal Mesh Component Object Reference=@n10.CustomBaseMesh) -> (Output_Get:Skeletal Mesh Component Object Reference)
   n11 = CheckForBackbling()
   > Set isWearingBackbling(isWearingBackbling=OR Boolean(A=@n11, B=@n11.bIsBackblingLong))
   > Set Is Backbling A Cape(Is Backbling A Cape:bool=@n11.bIsBackblingACape) -> (Output_Get:bool)
   n12 = Get bClothEnabled()
   > Set Cloth Weight Profile(Cloth State=@n12)
   > Set Prev Cloth Bool(Prev Cloth Bool=@n12)
   > DelayUntilNextFrame!(execute) -> (then)
   n13 = GetOwningComponent()
   > n14 = Cast To ChaosClothingInteractor!(
    execute
    Object=GetClothingInteractor!(
     self:Clothing Simulation Interactor Object Reference=GetClothingSimulationInteractor!(self:Skeletal Mesh Component Object Reference=@n13, ClothingAssetName:Name=None) -> (ReturnValue:Clothing Simulation Interactor Object Reference)
     ClothingAssetName:Name=F_MED_VowelSleep_Expand_Clothing
    ) -> (ReturnValue:Clothing Interactor Object Reference)
   ) -> (then, CastFailed, AsChaos Clothing Interactor:Chaos Clothing Interactor Object Reference) {is=title K2Node_DynamicCast}
   > ResetAndTeleport!(execute, self:Chaos Clothing Interactor Object Reference=@n14, bReset:bool=true, bTeleport:bool=true) -> (then)
   Event OnBeginSkydiving()
   > Set Anim Dynamics Gravity Override(Anim Dynamics Gravity Override:Vector="0,0,50") -> (Output_Get:Vector)
   > Set Anim Dynamics Alpha(Anim Dynamics Alpha=1)
   Event OnEndSkydiving()
   > Set Anim Dynamics Gravity Override(Anim Dynamics Gravity Override="0,0,-1800")
   > Set Anim Dynamics Alpha(Anim Dynamics Alpha=0.7)
   Event OnBeginPlayingEmote() {is=title K2Node_Event}
   > Set!(
    execute
    Cloth Core Settings Runtime_EdgeStiffness:Vector 2D Structure=(X=1,Y=1)
    Cloth Core Settings Runtime_BendingStiffness:Vector 2D Structure=(X=1,Y=1)
    Cloth Core Settings Runtime_AreaStiffness:Vector 2D Structure=(X=1,Y=1)
    Cloth Core Settings Runtime_TetherStiffness:Vector 2D Structure=(X=1.05,Y=1.05)
    Cloth Core Settings Runtime_CollisionThickness:float=2
    Cloth Core Settings Runtime_FrictionCoefficient:float=0.8
    Cloth Core Settings Runtime_SelfCollisionThickness:float=0
    Cloth Core Settings Runtime_DampingCoefficient:float=0.08
    Cloth Core Settings Runtime_LocalDampingCoefficient:float=0
    Cloth Core Settings Runtime_AnimDriveStiffness:Vector 2D Structure=(X=0.35,Y=1)
    Cloth Core Settings Runtime_AnimDriveDamping:Vector 2D Structure=(X=0.35,Y=1)
    Cloth Core Settings Runtime_NumberIterations:int=3
    Cloth Core Settings Runtime_NumberSubsteps:int=1
   ) -> (then, Output_Get:Cloth Core Settings Structure)
   > Set Anim Dynamics Alpha(Anim Dynamics Alpha=0.5)
   Event OnEndPlayingEmote() {is=title K2Node_Event}
   > Set Cloth Core Settings Runtime(Cloth Core Settings Runtime=Get Cloth Core Settings Default() -> (Cloth Core Settings Default:Cloth Core Settings Structure))
   > Set Anim Dynamics Alpha(Anim Dynamics Alpha=1)
   > Branch(Condition=Get bClothEnabled())
   n15 = GetOwningComponent()
   > n16 = Cast To ChaosClothingInteractor(Object=GetClothingInteractor(self=GetClothingSimulationInteractor(self=@n15, ClothingAssetName=None), ClothingAssetName=F_MED_VowelSleep_Expand_Clothing))
   > ResetAndTeleport(self=@n16, bReset=true, bTeleport=false)
   SetPressure!(execute, self:Chaos Clothing Interactor Object Reference, Pressure:Vector 2D Structure=(X=0.18,Y=1)) -> (then)
  ]}
 Handle Slope Warp {nodes [
   n1 = Handle Slope Warp!() -> (then, DeltaTime:double=~) {is=title K2Node_FunctionEntry}
   > n2 = Sequence!(execute) -> (then_0, then_1)
   n3 = Get bIsOnGround() -> (bIsOnGround:bool)
   > n4 = Branch(Condition=OR Boolean!(A=NOT Boolean(A=@n3), B=Get bIsPlayingEmote(), C:bool=Get bIsInVehicle() -> (bIsInVehicle:bool)) -> (ReturnValue))
   > Set Target Info Input Root Slope Translation(Target Info Input Root Slope Translation:Vector="0,0,-5000") -> (Output_Get:Vector)
   > Set Slope Info Input Slope Rotation(Slope Info Input Slope Rotation:Rotator="0, 0, 0") -> (Output_Get:Rotator)
   Set Slope Info Input Slope Rotation(
    execute=@n4.else
    Slope Info Input Slope Rotation=RInterpTo!(
     Current:Rotator=Get Slope Info Input Slope Rotation() -> (Slope Info Input Slope Rotation:Rotator)
     Target:Rotator=Get!() -> (
      SlopeInfoInput_RootSlopeTranslation:Vector=~
      SlopeInfoInput_LeftFootSlopeTranslation:Vector=~
      SlopeInfoInput_RightFootSlopeTranslation:Vector=~
      SlopeInfoInput_SlopeRotation:Rotator=~
     ).SlopeInfoInput_SlopeRotation
     DeltaTime:float=@n1
     InterpSpeed:float=7
    ) -> (ReturnValue:Rotator)
   )
   > Set Target Info Input Root Slope Translation(Target Info Input Root Slope Translation="0,0,-5")
   Set Slope Info Input Root Slope Translation(
    execute=@n2.then_1
    Slope Info Input Root Slope Translation:Vector=VInterpTo!(
     Current:Vector=Get Slope Info Input Root Slope Translation() -> (Slope Info Input Root Slope Translation:Vector)
     Target:Vector=Get Target Info Input Root Slope Translation() -> (Target Info Input Root Slope Translation:Vector)
     DeltaTime:float=@n1
     InterpSpeed:float=10
    ) -> (ReturnValue:Vector)
   ) -> (Output_Get:Vector)
  ]}
]
variables [
 Anim Dynamics Alpha {type=number, default=0.5}
 Anim Dynamics Gravity Override {type=Vector, default=<0 0 -1800>}
 Cloth Core Settings Default {
  type=ClothCoreSettings
  default {
   collisionThickness=1.5
   dampingCoefficient=0.08
   frictionCoefficient=0.3
   localDampingCoefficient=0
   numberIterations=3
   numberSubsteps=1
   selfCollisionThickness=0
   animDriveDamping=<0 1>
   animDriveStiffness=<0 1>
   areaStiffness=<1 1>
   bendingStiffness=<0.9 1>
   edgeStiffness=<0.9 1>
   tetherStiffness=<1 1>
  }
 }
 Cloth Core Settings Frontend {
  type=ClothCoreSettings
  default {
   collisionThickness=1.2
   dampingCoefficient=0.08
   frictionCoefficient=0.3
   localDampingCoefficient=0
   numberIterations=3
   numberSubsteps=1
   selfCollisionThickness=0
   animDriveDamping=<0 1>
   animDriveStiffness=<0 1>
   areaStiffness=<1 1>
   bendingStiffness=<0.9 1>
   edgeStiffness=<0.9 1>
   tetherStiffness=<1 1>
  }
 }
 Cloth Core Settings Runtime {
  type=ClothCoreSettings
  default {
   collisionThickness=1.5
   dampingCoefficient=0.08
   frictionCoefficient=0.3
   localDampingCoefficient=0
   numberIterations=3
   numberSubsteps=1
   selfCollisionThickness=0
   animDriveDamping=<0 1>
   animDriveStiffness=<0 1>
   areaStiffness=<1 1>
   bendingStiffness=<1 1>
   edgeStiffness=<1 1>
   tetherStiffness=<1 1>
  }
 }
 Custom Base Mesh {type=SkeletalMeshComponent}
 Delta Time X {type=number}
 Gravity Override {type=Vector}
 Is Backbling A Cape {type=boolean}
 isWearingBackbling {type=boolean}
 Prev Cloth Bool {type=boolean}
 Slope Info Input Root Slope Translation {type=Vector}
 Slope Info Input Slope Rotation {type=Rotator}
 Target Info Input Root Slope Translation {type=Vector}
 Wind Gust Local Settings {type=WindGustSettings, default {elapsedTime=0, finalGustStrengthLocal=0, gustDuration=0, gustStrength=0, timeBetweenGusts=0}}
]
