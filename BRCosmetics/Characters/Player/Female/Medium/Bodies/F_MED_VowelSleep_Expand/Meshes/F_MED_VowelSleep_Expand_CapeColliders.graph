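# Collision-only physics asset for the F_MED_VowelSleep_Expand female body, named as the set of colliders that capes collide against.
# Capsules wrap the pelvis, spine, clavicles and neck, with a sphere on the head.
# Large capsules on the dynamic rear jacket bones (center, left, right) let the collision volume follow the jacket.
# Arm and leg bodies carry no shapes, and no constraints link any bodies.
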PhysicsAsset /BRCosmetics/Characters/Player/Female/Medium/Bodies/F_MED_VowelSleep_Expand/Meshes/F_MED_VowelSleep_Expand_CapeColliders
bodies {
 calf_l []
 calf_r []
 clavicle_l [{shapeType=Capsule, length=7.9168, radius=6.0892, center=<6.3985 -1.4302 -1.3533>, rotation=<rot 75.0082 0.2776 175.831>}]
 clavicle_r [{shapeType=Capsule, length=7.9168, radius=6.0892, center=<-6.3985 1.4302 1.3533>, rotation=<rot 75.0082 0.2776 -4.169>}]
 dyn_jacket_rear_03 [{
   shapeType=Capsule
   length=20.1757
   radius=16.9082
   shapeName=dyn_jacket_rear_03_capsule
   center=<17.8726 6.9562 0>
   rotation=<rot 0 78.8776 -90>
  }]
 dyn_jacket_rear_05 [{shapeType=Capsule, length=20.5532, radius=19.3932, shapeName=dyn_jacket_rear_05_capsule, center=<15.787 0 0>, rotation=<rot 0 -90 -180>}]
 dyn_jacket_rear_l_03 [{
   shapeType=Capsule
   length=19.7392
   radius=11.1612
   shapeName=dyn_jacket_rear_l_03_capsule
   center=<17.7749 2.8995 4.6906>
   rotation=<rot 0 -91.4912 -85.6103>
  }]
 dyn_jacket_rear_l_05 [{
   shapeType=Capsule
   length=17.5566
   radius=12.3643
   shapeName=dyn_jacket_rear_l_05_capsule
   center=<15.7733 2.732 4.9696>
   rotation=<rot 0 -93.2077 -79.0792>
  }]
 dyn_jacket_rear_r_03 [{
   shapeType=Capsule
   length=19.3417
   radius=10.3952
   shapeName=dyn_jacket_rear_r_03_capsule
   center=<-17.8383 -2.7911 -4.145>
   rotation=<rot 0 88.819 -88.7574>
  }]
 dyn_jacket_rear_r_05 [{
   shapeType=Capsule
   length=16.4586
   radius=13.411
   shapeName=dyn_jacket_rear_r_05_capsule
   center=<-15.9598 -2.6482 -4.0926>
   rotation=<rot 0 -90.7934 -84.4372>
  }]
 foot_l []
 foot_r []
 head [{shapeType=Sphere, radius=17.3125, center=<11.028 -0.174 0>}]
 lowerarm_l []
 lowerarm_r []
 neck_02 [{shapeType=Capsule, length=13.4375, radius=7.3175, center=<-0.2642 -0.3408 0>, rotation=<rot 90 -171.1193 -175.3978>}]
 pelvis [
  {shapeType=Capsule, length=20.4217, radius=14.9115, center=<-13.2721 -4.5891 0>}
  {shapeType=Capsule, length=15, radius=15, shapeName=pelvis_capsule}
 ]
 spine_01 [{shapeType=Capsule, length=9.7497, radius=9.2108, center=<-1.7965 0.1884 0>}]
 spine_02 [
  {shapeType=Capsule, length=6.2037, radius=9.4443, center=<-1.711 1.2936 0.0001>}
  {shapeType=Capsule, length=29.3956, radius=11.5807, center=<2.5024 2.3084 3.9825>, rotation=<rot 90 0 0>}
  {shapeType=Capsule, length=31.3885, radius=9.5877, center=<2.4704 0.4991 -3.9825>, rotation=<rot 90 -0.0396 -0.0396>}
 ]
 spine_03 [{shapeType=Capsule, length=8.4663, radius=8.471, center=<0.1445 0.2396 0.0001>}]
 spine_04 [{shapeType=Capsule, length=4.6298, radius=11.8232, center=<0.9397 1.2948 0.0002>}]
 spine_05 [
  {shapeType=Capsule, length=7.7517, radius=11.9844, center=<-4.0281 1.2406 0>}
  {shapeType=Capsule, length=12.8499, radius=10.1327, center=<-0.2916 0.4452 0>}
 ]
 thigh_l []
 thigh_r []
 upperarm_l []
 upperarm_r []
}
constraints []
