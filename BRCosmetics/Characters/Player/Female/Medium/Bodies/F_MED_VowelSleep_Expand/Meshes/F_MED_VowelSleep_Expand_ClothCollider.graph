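# Physics asset that supplies cloth collision shapes for the VowelSleep_Expand medium female player body cosmetic.
# Capsules cover the pelvis, spine, feet and a ring of eight dynamic skirt bones, so collision volumes follow the skirt.
# A wide, flat box on def_slope_warp sits below that bone's origin, likely acting as a floor.
# Thigh and calf bodies hold no shapes, and no constraints link the bodies.
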PhysicsAsset /BRCosmetics/Characters/Player/Female/Medium/Bodies/F_MED_VowelSleep_Expand/Meshes/F_MED_VowelSleep_Expand_ClothCollider
bodies {
 calf_l []
 calf_r []
 def_slope_warp [{
   shapeType=Box
   extentX=286.2154
   extentY=278.7956
   extentZ=28.7369
   shapeName=def_slope_warp_box
   center=<0 0 -14.0088>
   rotation=<rot 0 -90 0>
  }]
 dyn_skirt_front_02 [{shapeType=Capsule, length=19.1244, radius=7.3717, shapeName=dyn_skirt_front_02_capsule, center=<2.5517 3.9053 0>, rotation=<rot 90 0 0>}]
 dyn_skirt_front_l_02 [{
   shapeType=Capsule
   length=17.6287
   radius=8.4559
   shapeName=dyn_skirt_front_l_02_capsule
   center=<2.4976 8.543 1.3182>
   rotation=<rot 77.0963 91.8958 90.6949>
  }]
 dyn_skirt_front_r_02 [{
   shapeType=Capsule
   length=17.6287
   radius=8.4559
   shapeName=dyn_skirt_front_r_02_capsule
   center=<-2.4976 -8.543 -1.3182>
   rotation=<rot 77.0963 91.8958 -89.3051>
  }]
 dyn_skirt_rear_02 [{
   shapeType=Capsule
   length=7.9376
   radius=11.9637
   shapeName=dyn_skirt_rear_02_capsule
   center=<6.9402 10.5332 -0.0691>
   rotation=<rot 1.742 -179.9999 -180>
  }]
 dyn_skirt_rear_l_02 [{
   shapeType=Capsule
   length=18.5203
   radius=9.1318
   shapeName=dyn_skirt_rear_l_02_capsule
   center=<4.3508 8.4122 0.9418>
   rotation=<rot 67.4932 -84.7049 -81.8374>
  }]
 dyn_skirt_rear_r_02 [{
   shapeType=Capsule
   length=18.5203
   radius=9.1318
   shapeName=dyn_skirt_rear_r_02_capsule
   center=<-4.3508 -8.4122 -0.9418>
   rotation=<rot 67.4932 -84.7049 98.1626>
  }]
 dyn_skirt_side_l_02 [{
   shapeType=Capsule
   length=17.19
   radius=9.0641
   shapeName=dyn_skirt_side_l_02_capsule
   center=<2.4127 8.6773 -0.3511>
   rotation=<rot 90 3.525 0>
  }]
 dyn_skirt_side_r_02 [{
   shapeType=Capsule
   length=17.19
   radius=9.0641
   shapeName=dyn_skirt_side_r_02_capsule
   center=<-2.4072 -8.7668 0.3511>
   rotation=<rot 90 -176.475 0>
  }]
 foot_l [{
   shapeType=Capsule
   length=10.9838
   radius=13.1602
   shapeName=foot_l_capsule
   center=<-1.9446 6.2339 -0.1526>
   rotation=<rot 0.6166 -2.3616 -87.7705>
  }]
 foot_r [{
   shapeType=Capsule
   length=10.9838
   radius=13.1602
   shapeName=foot_r_capsule
   center=<1.9446 -6.2339 0.1526>
   rotation=<rot 0.6166 -2.3616 92.2295>
  }]
 pelvis [{shapeType=Capsule, length=14.4912, radius=8.2406, shapeName=pelvis_capsule, center=<0 -1.2975 0>}]
 spine_02 [{shapeType=Capsule, length=11.7365, radius=11.9184, shapeName=spine_02_capsule, center=<-9.0523 0.938 0>, rotation=<rot 90 4.4892 0>}]
 spine_04 [{shapeType=Capsule, length=10.4771, radius=11.5229, shapeName=spine_04_capsule, center=<-5.6048 2.4001 0>, rotation=<rot -90 4.6177 0>}]
 thigh_l []
 thigh_r []
}
constraints []
